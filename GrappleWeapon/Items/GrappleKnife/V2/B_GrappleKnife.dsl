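; Fortnite weapon Blueprint for the Grapple Knife, handling the presentation and state of its grapple and melee attacks.
; On equip it attaches left-hand meshes, binds ammo, attach and movement events, and starts a locally viewed timer that polls targets for the reticle.
; Fire, zip, bounce and attack events drive reticle state, effects and a replicated anim enum.
; The enum's OnRep sets one-shot animation bools, cleared next tick.
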(event Custom|Event%20On%20Fired
  (params (DistanceBucket :type Integer) (WindUpTime :type Float%20%28double-precision%29) (ProjectileTravelTime :type Float%20%28double-precision%29))
  ;[1 1 1] Update reticle on fire
  (Variables|ReplicatedVariables|SetDistanceBucket DistanceBucket)
  (Variables|Default|SetWindUpTime WindUpTime)
  (Variables|Default|SetProjectileTravelTime ProjectileTravelTime)
  (Variables|Default|SetCachedInputDirection (Default|GetInputDirection))
  (CallFunction|SetReticleState "NewEnumerator2")
  (Variables|ReplicatedVariables|SetReplicatedAnimState "NewEnumerator7")
  ;[1 1 1] Fail safe for grapple hook not attaching
  (Utilities|FlowControl|Delay 1.0)
  (if (Utilities|Enum|Equal(Enum) (Variables|Default|GetReticleState) "NewEnumerator2")
    (Variables|Default|SetReticleState "NewEnumerator0")))

(event Custom|Event%20Update%20Reticle%20Targeting
  ;[1 1 1] On Equip event update reticle and attach weapon to pawn and bind events
  (CallFunction|UpdateReticleTargeting))

(event FortWeapon|EventOnEquippedAndAttached
  (bind _as_fort_player_pawn (Variables|Default|GetAs%20Fort%20Player%20Pawn))
  ;[1 1 1] On Equip event update reticle and attach weapon to pawn and bind events
  (|CallParentFunction)
  (bind _asfort_player_pawn (Utilities|Casting|CastToFortPlayerPawn (Actor|GetOwner)))
  (bind _output_get (Variables|Default|SetAs%20Fort%20Player%20Pawn _asfort_player_pawn))
  (Utilities|FlowControl|Sequence
    (:then_0
      (if (Pawn|IsLocallyViewed _output_get)
        (bind _returnvalue (Utilities|Time|SetTimerbyEvent (delegate Custom|Event%20Update%20Reticle%20Targeting) 0.01 true true))
        (Variables|Default|SetReticleTargetingTimer _returnvalue)))
    (:then_1
      (Utilities|IsValid (Variables|/Script/Engine.Character|GetMesh _output_get)
        (:Is%20Valid
          (Transformation|AttachComponentToComponent (Variables|Default|GetLeftHandStaticMesh) (Variables|/Script/Engine.Character|GetMesh _output_get) (Variables|Default|GetLeftHandAttachSocket) "SnapToTarget" "SnapToTarget" "KeepRelative" true)
          (Transformation|AttachComponentToComponent (Variables|Default|GetLeftHandSkeletalMesh) (Variables|/Script/Engine.Character|GetMesh _output_get) (Variables|Default|GetLeftHandAttachSocket) "SnapToTarget" "SnapToTarget" "KeepRelative" true))))
    (:then_2
      (Variables|ReplicatedVariables|SetIsGrappleArmed (>= (FortWeapon|GetLocalMagazineAmmoCount) 1))
      (Variables|Default|SetPreviousAmmoCount (FortWeapon|GetLocalMagazineAmmoCount)))
    (:then_3
      (Ability|Async|WaitGameplayEventToActor _as_fort_player_pawn "(TagName=\"Event.GrappleKnife.Secondary.Attached\")"
        (:EventReceived
          (Variables|ReplicatedVariables|SetIsAttached true))))
    (:then_4
      (Ability|Async|WaitGameplayEventToActor _as_fort_player_pawn "(TagName=\"Event.GrappleKnife.Secondary.Detached\")"
        (:EventReceived
          (Variables|ReplicatedVariables|SetIsAttached false))))
    (:then_5
      (EventDispatchers|AssignOnLocalAmmoChanged (EventDispatchers|CreateEvent :!event Custom|OnLocalAmmoChanged_Event)))
    (:then_6
      (EventDispatchers|BindEventtoZiplineStateChanged _as_fort_player_pawn (delegate Custom|OnZipLineEvent))
      (Ability|Tasks|WaitGameplayTagAddToActor (Variables|Default|GetAs%20Fort%20Player%20Pawn) "(TagName=\"MovementMode.Grinding\")"
        (:Added
          (CallFunction|RemoveQuestTravelingTag))))))

(event Custom|EventOnEnter
  (params (Vehicle :type Fort%20Vehicle%20Interface%20Interface%20%28by%20ref%29) (PlayerPawn :type Fort%20Pawn%20Object%20Reference) (SeatIndex :type Integer))
  ;[1 1 1] EventOnEnter/Exit Vehicle
  (CharacterPart|SetCharacterPartVisibility (Variables|Default|GetAs%20Fort%20Player%20Pawn) "Gameplay" false true "()"))

(event Custom|EventOnExit
  (params (Vehicle :type Fort%20Vehicle%20Interface%20Interface%20%28by%20ref%29) (PlayerPawn :type Fort%20Pawn%20Object%20Reference) (SeatIndex :type Integer))
  ;[1 1 1] EventOnEnter/Exit Vehicle
  (CharacterPart|SetCharacterPartVisibility (Variables|Default|GetAs%20Fort%20Player%20Pawn) "Gameplay" true true "()"))

(event Custom|OnZipLineEvent
  (params (bIsZiplining :type Boolean) (FortPlayerPawn :type Fort%20Player%20Pawn%20Object%20Reference))
  ;[1 1 1] On Equip event update reticle and attach weapon to pawn and bind events
  (if bIsZiplining
    (CallFunction|RemoveQuestTravelingTag)))

(event Custom|OnStartZip
  (params (ZipDistance :type Float%20%28double-precision%29))
  ;[1 1 1] Update reticle on zip start
  (CallFunction|SetReticleState "NewEnumerator3" ZipDistance)
  (Variables|ReplicatedVariables|SetReplicatedAnimState "NewEnumerator8"))

(event Custom|OnEndZip
  (params (Bounced :type Boolean) (BounceAngle :type Float%20%28double-precision%29))
  ;[1 1 1] On zip end update reticle
  (CallFunction|SetReticleState "NewEnumerator4")
  (if Bounced
    (Debugging|PrintFloat BounceAngle "End Zip Bounce Angle "
      (:Then
        (GameplayCue|ExecuteGameplayCueOnActor(Burst) (Variables|Default|GetAs%20Fort%20Player%20Pawn) (Variables|Default|GetGC_BounceImpact) (Ability|GameplayCue|MakeGameplayCueParameters 0.0 BounceAngle))
        (Variables|ReplicatedVariables|SetAnim_BounceAngle BounceAngle)
        (Variables|ReplicatedVariables|SetReplicatedAnimState "NewEnumerator9")
        (Utilities|FlowControl|Delay 1.0)
        (if (Utilities|Enum|Equal(Enum) (Variables|ReplicatedVariables|GetReplicatedAnimState) "NewEnumerator9")
          _)))
    (else
      _))
  (Variables|ReplicatedVariables|SetReplicatedAnimState "NewEnumerator6"))

(event Custom|OnLocalAmmoChanged_Event
  (params (LocalCount :type Integer) (LocalRemaining :type Integer))
  ;[1 1 1] Updated grapple armed states (plays sfx/changes anim state)
  (Variables|ReplicatedVariables|SetIsGrappleArmed (>= LocalCount 1))
  (if (> LocalCount (Variables|Default|GetPreviousAmmoCount))
    (if (== LocalCount 1)
      (Audio|PlaySoundatLocation (Variables|Default|GetReadyStateAudio) (Transformation|GetActorLocation))
      (CallFunction|SetReticleState (Variables|Default|GetReticleState))
      (else
        _))
    (Variables|Default|SetPreviousAmmoCount LocalCount)
    (else
      _)))

(event FortWeapon|EventOnUnEquippedAndDetached
  ;[1 1 1] Attach weapon to the left hand
  (|CallParentFunction)
  (Utilities|Time|ClearandInvalidateTimerbyHandle (Variables|Default|GetReticleTargetingTimer))
  (Transformation|DetachFromComponent (Variables|Default|GetLeftHandStaticMesh) "KeepRelative" "KeepRelative" "KeepRelative" true)
  (CallFunction|SetReticleState "NewEnumerator0")
  (CallFunction|RemoveQuestTravelingTag))

(event Custom|RemoveQuestTravelingTag
  ;[1 1 1] Remove quest traveling tag
  (Ability|GameplayEffect|RemoveLooseGameplayTags (Variables|Default|GetAs%20Fort%20Player%20Pawn) (Variables|Default|GetQuest_TravelingTag) true))

(event EventOnAnimTrailsNotify
  (params (bActive :type Boolean))
  ;[1 1 1] Event anim trials
  (if bActive
    (Networking|CanExecuteCosmeticEvents
      (:True
        (Components|Activation|Activate (Variables|Default|GetLeftHandTrail) true)))
    (else
      (Networking|CanExecuteCosmeticEvents
        (:True
          (Components|Activation|Deactivate (Variables|Default|GetLeftHandTrail)))))))

(event EventOnAnimTrailsNotify(Deactivate)
  ;[1 1 1] Event anim trials
  (Networking|CanExecuteCosmeticEvents
    (:True
      (Components|Activation|Deactivate (Variables|Default|GetLeftHandTrail)))))

(event FortWeapon|EventOnPlayImpactFX
  (params (HitResult :type Hit%20Result%20Structure%20%28by%20ref%29) (ImpactPhysicalSurface :type EPhysicalSurface%20Enum :default "SurfaceType_Default") (SpawnedPSC :type FXSystem%20Component%20Object%20Reference))
  (bind (_bistechniqueactive _activetechniquename) (Utilities|Struct|BreakFortMantisReplicatedAnimInstanceInfo (Mantis|GetAnimInstanceInfo (Variables|Default|GetFortMantisWeapon))))
  (bind (_bblockinghit _binitialoverlap _time _distance _location _impactpoint _normal _impactnormal _physmat _hitactor) (Collision|BreakHitResult HitResult))
  ;[0 0.4844 0.1657 1 50] DIRECTIONAL IMPACT
  (|CallParentFunction HitResult ImpactPhysicalSurface SpawnedPSC)
  (Utilities|IsValid SpawnedPSC
    (:Is%20Valid
      (Transformation|SetWorldRotation SpawnedPSC (Math|Rotator|CombineRotators (Math|Rotator|MakeRotfromX _normal) (select (Utilities|String|EqualExactly(String) (Utilities|String|ToString(Name) _activetechniquename) "FullBodyAttack1") (Variables|Default|GetRight%20Rotation) (Variables|Default|GetLeft%20Rotation))) false false))
    (:Is%20Not%20Valid
      _))
  (Utilities|Casting|CastToFortPlayerPawn _hitactor
    (:CastFailed
      (Audio|PlaySoundatLocation "/GrappleWeapon/Assets/Audio/GrappleKnife_Hit_Default_Cue.GrappleKnife_Hit_Default_Cue" _impactpoint))))

(event Custom|StartAttack
  (params (Is%20in%20Air%20Attack :type Boolean))
  (bind _lefthandtrail (Variables|Default|GetLeftHandTrail))
  ;[1 1 1] Start attack reticle state
  (CallFunction|SetReticleState "NewEnumerator5")
  (Components|Activation|Deactivate (Variables|Default|GetLeftHandIdleFX))
  (if Is%20in%20Air%20Attack
    (Niagara|SetNiagaraVariable(Bool) _lefthandtrail "is360Atk" true)
    (Niagara|SetNiagaraVariable(Vector3) _lefthandtrail "ParentVelocity" (Transformation|GetVelocity (Variables|Default|GetAs%20Fort%20Player%20Pawn)))
    (Variables|/Script/Engine.PostProcessComponent|SetbEnabled true (Variables|Default|GetPostProcess))
    (bind (_direction _lerp) (Timeline|Timeline ReverseFromEnd :NewTime 1.0)
      (:Update
        (Variables|/Script/Engine.PostProcessComponent|SetBlendWeight _lerp (Variables|Default|GetPostProcess))))
    (else
      (Niagara|SetNiagaraVariable(Bool) _lefthandtrail "is360Atk" false))))

(event Custom|EndAttack
  ;[1 1 1] End attack reticle state
  (CallFunction|SetReticleState "NewEnumerator6")
  (Components|Activation|Activate (Variables|Default|GetLeftHandIdleFX) false)
  (Niagara|SetNiagaraVariable(Bool) (Variables|Default|GetLeftHandTrail) "is360Atk" false))

(event Custom|ForceField_InAirAtk
  ;[1 0.09855 0.4353 1 50] FORCE FIELD WHEN IN AIR ATTACK (water + grass)
  (Game|SpawnActorfromClass "/Game/Effects/Niagara/Blueprints/PhsyicsFields/FN_RadialForce.FN_RadialForce_C" (Transformation|GetActorTransform) "Undefined" "MultiplyWithRoot" 0 0.2 0.6 4e+03 1.2 0.0 0))

(event Custom|NoGrappleTarget
  ;[1 1 1] Grapple target not found
  (Development|PrintString "NoTarget")
  (Variables|ReplicatedVariables|SetReplicatedAnimState "NewEnumerator10")
  (Utilities|FlowControl|Delay 1.0)
  (if (Utilities|Enum|Equal(Enum) (Variables|ReplicatedVariables|GetReplicatedAnimState) "NewEnumerator6")
    (Variables|ReplicatedVariables|SetReplicatedAnimState "NewEnumerator6")))

(event Custom|DebugNoTarget
  (params (NoTargetDebug :type Boolean))
  ;[1 1 1] No target debug
  (Variables|Default|SetNoTargetDebug NoTargetDebug))

(event Custom|NoAmmoFire
  (Default|CallOnNoAmmoFire))

(fn ClearAnimStateBools
  (Variables|AnimStateBools|SetIsFired false)
  (Variables|AnimStateBools|SetIsLaunched false)
  (Variables|AnimStateBools|SetAnim_StartWallBounce false)
  (Variables|AnimStateBools|SetbHasNoTarget false))

(fn GetGrappleKnifeData
  (return (Variables|ReplicatedVariables|GetIsAttached) (Variables|AnimStateBools|GetIsFired) (Variables|ReplicatedVariables|GetDistanceBucket) (Variables|AnimStateBools|GetAnim_StartWallBounce) (Variables|ReplicatedVariables|GetAnim_BounceAngle)))

(fn GetInputDirection
  (bind (_x _y) (Math|Vector2D|BreakVector2D (Math|Vector2D|NormalSafe(Vector2D) (Variables|/Script/FortniteGame.FortPlayerController|GetLastPlayerLocalMoveInput (Fortnite|GetFortPlayerControllerfromActor (Actor|GetOwner))) 1e-05)))
  (return (Math|Integer|SelectInt -1 (Math|Integer|SelectInt 1 0 (Math|Float|InRange(Float) _y 0.5)) (Math|Float|InRange(Float) _y -1.0 -0.5))))

(fn OnRep_ReplicatedAnimState
  (CallFunction|ClearAnimStateBools)
  (switch Utilities|FlowControl|Switch|SwitchonGrappleKnifeWeaponAnimState (Variables|ReplicatedVariables|GetReplicatedAnimState)
    (:NewEnumerator7
      (Variables|AnimStateBools|SetIsFired true))
    (:NewEnumerator8
      (Variables|AnimStateBools|SetIsLaunched true))
    (:NewEnumerator9
      (Variables|AnimStateBools|SetAnim_StartWallBounce true))
    (:NewEnumerator10
      (Variables|AnimStateBools|SetbHasNoTarget true)))
  (Utilities|Time|SetTimerforNextTickbyFunctionName 0 "ClearAnimStateBools"))

(fn SetReticleState
  (params (NewReticleState :type GrappleKnifeReticleState%20Enum :default "NewEnumerator0") (ZipDistance :type Float%20%28double-precision%29) (NewWorldLocation :type Vector) (bTargetingLedge :type Boolean))
  (bind _output_get (Variables|Default|SetReticleWorldLocation (Math|Vector|SelectVector (Variables|Default|GetReticleWorldLocation) NewWorldLocation (Math|Vector|VectorIsZero NewWorldLocation))))
  (bind _output_get_2 (Variables|Default|SetReticleState (Variables|GetNewReticleState)))
  (Default|CallOnGrappleKnifeReticleStateChanged _output_get_2 (Variables|GetZipDistance) _output_get bTargetingLedge)
  (Development|PrintString (Utilities|String|Append "Grapple Knife Reticle: " (Utilities|String|EnumtoString (Variables|Default|GetReticleState)))))

(fn UpdateReticleTargeting
  (bind _returnvalue (Game|GetInstigator))
  (bind _asfort_player_pawn (Utilities|Casting|CastToFortPlayerPawn _returnvalue))
  (if (Pawn|IsLocallyViewed _returnvalue)
    (switch Utilities|FlowControl|Switch|SwitchonGrappleKnifeReticleState (Variables|Default|GetReticleState)
      (:NewEnumerator0
        _)
      (:NewEnumerator1
        _)
      (:NewEnumerator7
        _)
      (:NewEnumerator4
        _)
      (:NewEnumerator6
        _))
    (bind (_bhittarget _targetactor _targetcomponent _targetpoint _hitdistance _btargetledge) (Default|GrappleKnifeGetTargetInfo _asfort_player_pawn))
    (if _bhittarget
      (Utilities|Casting|CastToFortPlayerPawn _targetactor
        (:then
          (CallFunction|SetReticleState "NewEnumerator7" _hitdistance _targetpoint _btargetledge))
        (:CastFailed
          (CallFunction|SetReticleState "NewEnumerator1" _hitdistance _targetpoint _btargetledge)))
      (else
        (CallFunction|SetReticleState "NewEnumerator0")))))

(fn ConstructionScript
  (|CallParentFunction)
  (Niagara|SetNiagaraSystemAsset (Variables|Default|GetLeftHandTrail) (Variables|Effects|GetAnimTrail%20Niagara%20Template) true)
  (Niagara|SetNiagaraSystemAsset (Variables|Default|GetLeftHandIdleFX) (Variables|Effects|GetIdleFX%20Niagara%20Template) true))
