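; Grapple Knife hook projectile that sticks to what it hits and starts the player's zip toward it.
; The server spawns a linked rope on begin play; impact tightens it and attaches the hook to the surface or object.
; Server-side travel caches a distance-based zip duration on the weapon, signals the pawn through gameplay events, then retracts the rope on a timer.
; Skydive launches, weapon switches and DBNO trigger early cleanup.
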(event Custom|Grapple
  (params (Hit :type Hit%20Result%20Structure%20%28by%20ref%29))
  (bind _playerpawn (Variables|Default|GetPlayerPawn))
  ;[1 1 1] Grapple - Calculate the launch values, blocking and clearing movement as needed. Plays Visuals for lense effects and shakes.
  (GameplayCue|ExecuteGameplayCueOnActor(Burst) (Variables|Default|GetPlayerPawn) (Variables|Default|GetGC_Grappled) (Ability|GameplayCue|MakeGameplayCueParameters 0.0))
  (if (not (Variables|Default|GetAirControl))
    (Input|SetIgnoreMoveInput (Game|GetInstigatorController (Variables|Default|GetPlayerPawn)) true))
  (HandleSticks|HandleMomentum)
  (FortMovement|CancelJumpAbility (Variables|Default|GetPlayerPawn) false)
  ;[1 1 1] Movement and camera modifier
  (Ability|SendGameplayEventtoActor (Variables|Default|GetPlayerPawn) "(TagName=\"Event.GrappleKnife.Secondary.Attached\")")
  (CallFunction|CacheZipDurationonWeapon)
  (Networking|SwitchHasAuthority
    (:Authority
      (GameplayEffects|RemoveActiveGameplayEffectbySourceEffect (Ability|GetAbilitySystemComponent (Variables|Default|GetPlayerPawn)) "/GrappleWeapon/Items/GrappleKnife/V2/GE_GrappleKnife_BlockPrimaryTag.GE_GrappleKnife_BlockPrimaryTag_C" (Ability|GetAbilitySystemComponent (Variables|Default|GetPlayerPawn)) -1))
    (:Remote
      (Ability|SendGameplayEventtoActor _playerpawn "(TagName=\"Event.GrappleKnife.Secondary.ZipToMovement\")" (Utilities|Struct|MakeGameplayEventData "(TagName=\"\")" 0 self 0 0 "()" "(GameplayTags=)" "(GameplayTags=)" 0.0 (Ability|TargetData|AbilityTargetDatafromHitResult Hit)))
      (Ability|SendGameplayEventtoActor _playerpawn "(TagName=\"Event.GrappleKnife.Secondary.CameraModifier.Add\")"))))

(event Custom|LaunchCheck
  ;[1 1 1] Reset everything if a launchpad is touched while Proj is active
  (bind _jumppadding (CallFunction|HandleLaunch))
  (if _jumppadding
    (Utilities|FlowControl|DoOnce false
      (:Completed
        (Variables|Default|SetLaunchPad true)
        (Variables|Default|SetHookStopped true)
        (Ability|SendGameplayEventtoActor (Variables|Default|GetPlayerPawn) "(TagName=\"Abilities.Grapple.Hit\")" (Utilities|Struct|MakeGameplayEventData :Instigator self :ContextHandle "()" :InstigatorTags "(GameplayTags=)" :TargetTags "(GameplayTags=)" :EventMagnitude 0.0 :TargetData "()"))
        (bind _rope (Variables|Default|GetRope)
          (:then
            (Actor|DestroyActor _rope)
            (Utilities|FlowControl|Delay 0.1) ; so proj and rope can't be destroyed at the same time
            (Actor|DestroyActor))
          (:else))))))

(event Custom|WeaponCheck
  ;[1 1 1] Destroys projectil if weapon is switched.  This is so we can keep the rope FX looking good, but still let people weapon switch after grappling
  (if (not (Utilities|Equal(Object) (Variables|/Script/FortniteGame.FortPawn|GetCurrentWeapon (Variables|Default|GetPlayerPawn)) (Variables|Default|GetAs%20B%20Grapple%20Knife)))
    (Actor|DestroyActor)))

(event EventBeginPlay
  (bind _playerpawn (Variables|Default|GetPlayerPawn))
  (bind _returnvalue (Fortnite|GetFortPlayerControllerfromActor _playerpawn))
  (bind _asfort_player_controller_athena (Utilities|Casting|CastToFortPlayerControllerAthena _returnvalue))
  ;[1 1 1] Begin Play
  (bind _asfort_player_pawn (Utilities|Casting|CastToFortPlayerPawn (Variables|Actor|GetInstigator)))
  (Variables|Default|SetPlayerPawn _asfort_player_pawn)
  (Tags|AddNonReplicatedTagtoPawn (Variables|Default|GetPlayerPawn) (Variables|Default|GetRopeActiveTag))
  (Utilities|IsValid _asfort_player_controller_athena
    (:Is%20Valid
      (EventDispatchers|BindEventtoOnBeginSkydiving _asfort_player_controller_athena (EventDispatchers|CreateEvent :!event Custom|LaunchCheck))
      (bind _asb_grapple_knife (Utilities|Casting|CastToB_GrappleKnife (Actor|GetOwner)))
      (Variables|Default|SetAs%20B%20Grapple%20Knife _asb_grapple_knife)
      (bind _playerpawn_2 (Variables|Default|GetPlayerPawn)
        (:then
          (if (Networking|IsServer)
            (CallFunction|SpawnRope _playerpawn_2)
            (bind _returnvalue_2 (GameplayEffects|ApplyGameplayEffectToTarget (Ability|GetAbilitySystemComponent _playerpawn_2) (Variables|Default|GetGE_RopeActive) (Ability|GetAbilitySystemComponent _playerpawn_2) 1.0 (Utilities|Struct|MakeGameplayEffectContextHandle)))
            (bind _rope (Variables|Default|GetRope)
              (:then
                (Variables|/GrappleWeapon/Items/GrappleKnife/V2/B_GrappleKnife_Rope.B_GrappleKnife_Rope_C|SetRopeGE _returnvalue_2 _rope))
              (:else))
            (else
              _))
          (Variables|Default|SetFireAImPitch (FortPlayerPawn|GetBaseAimPitchQuantized (Variables|Default|GetPlayerPawn)))
          (Tags|AddNonReplicatedTagtoPawn (Variables|Default|GetPlayerPawn) (Variables|Default|GetBlockIncomingPickups)))
        (:else)))))

(event Collision|EventHit
  (params (MyComp :type Primitive%20Component%20Object%20Reference) (Other :type Actor%20Object%20Reference) (OtherComp :type Primitive%20Component%20Object%20Reference) (bSelfMoved :type Boolean) (HitLocation :type Vector :default "0, 0, 0") (HitNormal :type Vector :default "0, 0, 0") (NormalImpulse :type Vector :default "0, 0, 0") (Hit :type Hit%20Result%20Structure%20%28by%20ref%29))
  (bind (_bblockinghit _binitialoverlap _time _distance _location _impactpoint _normal _impactnormal _physmat _hitactor) (Collision|BreakHitResult Hit))
  ;[1 1 1] Stop event.  Sets variables and calls the travel event
  (if (Utilities|NotEqual(Object) self _hitactor)
    (Variables|Default|SetHit%20Actor _hitactor)
    (bind _playerpawn (Variables|Default|GetPlayerPawn)
      (:then
        (Tether|SetTetherPawn _playerpawn 0)
        (if (Variables|/Script/FortniteGame.FortPawn|GetbIsDBNO _playerpawn)
          (bind _playerpawn_2 (Variables|Default|GetPlayerPawn)
            (:then
              (Ability|SendGameplayEventtoActor _playerpawn_2 (Variables|Default|GetEvent_Returned)) ; Unblock pawn from weapon switching when interp finishes
              (bind _rope (Variables|Default|GetRope)
                (:then
                  (Rendering|SetActorHiddenInGame _rope true)
                  (Actor|DestroyActor _rope))
                (:else)))
            (:else))
          (else
            (CallFunction|HandleHits Hit)
            (Actor|SetLifeSpan (Variables|Default|GetLifespanOnHit))
            (CallFunction|StartReel)
            (if (and (Networking|IsServer) (Utilities|IsValid (Game|GetInstigatorController) :!class /Script/Engine.KismetSystemLibrary))
              (CallFunction|Travel Hit)))))
      (:else))))

(event Custom|Travel
  (params (Hit :type Hit%20Result%20Structure))
  (bind (_bblockinghit _binitialoverlap _time _distance _location _impactpoint) (Collision|BreakHitResult Hit))
  (bind _playerpawn (Variables|Default|GetPlayerPawn))
  ;[1 1 1] Travel event.  Plays travel montage, sounds and then moves the player.
  (Variables|Default|SetHitLocation _impactpoint)
  (Utilities|IsValid (Variables|Default|GetAs%20B%20Grapple%20Knife)
    (:Is%20Valid
      _)
    (:Is%20Not%20Valid
      _))
  (bind _jumppadding (CallFunction|HandleLaunch))
  (if _jumppadding
    (CallFunction|LaunchCheck) ; client/server run into disagreements here - this is a failsafe
    (else
      (bind _returnvalue (Zipline|IsZiplining _playerpawn))
      (if _returnvalue
        (Zipline|EndZiplining _playerpawn false false)
        (else
          _))
      (Ability|GameplayEffect|AddLooseGameplayTags (Variables|Default|GetPlayerPawn) (Variables|Default|GetBlockGrindingTag))
      (Variables|/Script/Engine.Actor|GetRootComponent (Variables|Default|GetPlayerPawn)
        (:then
          _)
        (:else
          _))
      (Ability|SendGameplayEventtoActor (Variables|Default|GetPlayerPawn) (Variables|Default|GetEvent_GrappleHit) (Utilities|Struct|MakeGameplayEventData :Instigator self :ContextHandle "()" :InstigatorTags "(GameplayTags=)" :TargetTags "(GameplayTags=)" :EventMagnitude 0.0 :TargetData "()"))
      (CallFunction|Grapple Hit)
      (Utilities|Time|SetTimerbyFunctionName 0 "WeaponCheck" 0.1 true) ; destroy projectile if weapons switched
      (CallFunction|GiveQuestCredit)
      (Utilities|Time|SetTimerbyEvent (delegate Custom|Retract%20Rope) (Variables|Default|GetDelayBeforeRetract)))))

(event OnEnterWaterEvent(WaterInteractionComponent)
  (params (WaterBody :type Fort%20Water%20Body%20Actor%20Object%20Reference) (WaterInteractionComponent :type Fort%20Water%20Interaction%20Component%20Object%20Reference) (bIsFirstBody :type Boolean))
  (bind _waterinteractioncomponent (Variables|Water|GetWaterInteractionComponent))
  (bind (_waterplanelocation _waterplanenormal _watersurfaceposition _waterdepth _waterbodyidx) (FX|GetLastWaterSurfaceInfo _waterinteractioncomponent :!class /Script/FortniteGame.FortWaterInteractionComponent))
  ;;[1 1 1] Water splash
  ;[1 0 0] Disallow usage in water
  (if bIsFirstBody
    (bind _returnvalue (Effects|SpawnEmitteratLocation (Variables|Default|GetFX_WaterImpact) _watersurfaceposition (Math|Rotator|MakeRotfromX _waterplanenormal)))
    (Effects|Components|ParticleSystem|SetFloatParameter _returnvalue "WaterDepth" _waterdepth)
    (Effects|Components|ParticleSystem|SetFloatParameter _returnvalue "WaterBodyIndex" (Math|Conversions|ToFloat(Integer) _waterbodyidx))
    (Audio|SpawnSoundatLocation (Variables|GrapplerSounds|GetSound_Water) _waterplanelocation)))

(event Custom|Retract%20Rope
  ;[1 1 1] Travel event.  Plays travel montage, sounds and then moves the player.
  (CallFunction|StartRetract)
  ;;[1 1 1] Retract Time
  (Utilities|FlowControl|Delay 0.5)
  (bind _playerpawn (Variables|Default|GetPlayerPawn)
    (:then
      (GameplayEffects|RemoveActiveGameplayEffectbySourceEffect (Ability|GetFortAbilitySystemComponent _playerpawn) (Variables|Default|GetGE_RopeActive) (Ability|GetFortAbilitySystemComponent _playerpawn) -1))
    (:else)))

(event Custom|StartReel
  ;[1 1 1] Reeling and retracting
  ;;[1 1 1] Timer to match with retract
  (Utilities|FlowControl|Delay 0.2)
  (bind _rope (Variables|Default|GetRope)
    (:then
      (Variables|/Script/FortniteGame.FortHookGunRope|SetbProjectileGone true _rope))
    (:else)))

(event EventDestroyed
  ;[1 1 1] Destroyed
  (Ability|GameplayEffect|RemoveLooseGameplayTags (Variables|Default|GetPlayerPawn) (Variables|Default|GetBlockGrindingTag))
  (bind _playerpawn (Variables|Default|GetPlayerPawn)
    (:then
      (if (Variables|/Script/FortniteGame.FortPawn|GetbIsDBNO _playerpawn)
        (Ability|SendGameplayEventtoActor (Variables|Default|GetPlayerPawn) (Variables|Default|GetEvent_Returned)) ; Unblock pawn from weapon switching when interp finishes
        (bind _rope (Variables|Default|GetRope)
          (:then
            (Actor|DestroyActor _rope))
          (:else))
        (elif (Variables|Default|GetHookStopped)
          (bind _rope_2 (Variables|Default|GetRope)
            (:then
              (Actor|DestroyActor _rope_2))
            (:else))
          (else
            ;;[1 0 0] No valid targets (if target becomes invalid before grappling)
            (Class|BGrappleKnife|NoGrappleTarget (Variables|Default|GetAs%20B%20Grapple%20Knife))
            (Variables|Default|GetPlayerPawn
              (:then
                (bind _rope_3 (Variables|Default|GetRope)
                  (:then
                    (Variables|/Script/FortniteGame.FortHookGunRope|SetbProjectileGone true _rope_3))
                  (:else)))
              (:else))))))
    (:else)))

(event Custom|StartRetract
  ;[1 1 1] Reeling and retracting
  (Development|SetHiddeninGame (Variables|Default|GetProjectileMesh) true true)
  (Ability|SendGameplayEventtoActor (Variables|Default|GetPlayerPawn) "(TagName=\"Event.GrappleKnife.Secondary.Detached\")"))

(fn CacheZipDurationOnWeapon
  ;[1 0 0.009724] Need to move this logic over to the movement GA
  (Variables|Default|SetDistance (Math|Vector|Distance(Vector) (Transformation|GetActorLocation (Variables|Default|GetPlayerPawn)) (Transformation|GetActorLocation)))
  (Variables|/GrappleWeapon/Items/GrappleKnife/V2/B_GrappleKnife.B_GrappleKnife_C|SetZipDuration (Math|Float|Clamp(Float) (Math|Float|SafeDivide (Variables|Default|GetDistance) (Variables|Default|GetBaseZipSpeed)) (Variables|Default|GetMinZipTime) (Variables|Default|GetMaxZipTime)) (Variables|Default|GetAs%20B%20Grapple%20Knife)))

(fn CalculateYankForce
  (bind (_outlocation _outrotation) (Actor|GetActorEyesViewPoint (Variables|Default|GetPlayerPawn)))
  (bind _outlaunchvelocity (Game|SuggestProjectileVelocityCustomArc (Transformation|GetActorLocation (Variables|Default|GetTarget)) (+ _outlocation (* (Math|Vector|GetForwardVector _outrotation) (Variables|Default|GetDistanceFromInstigator))))) ; Finds out how much force is needed to put the target 300 units in front of where instigator is facing
  (if (Utilities|Enum|NotEqual(Enum) (Networking|GetLocalRole (Variables|Default|GetTarget)) "ROLE_SimulatedProxy")
    (Character|LaunchCharacter (Variables|Default|GetTarget) (+ _outlaunchvelocity (Math|Vector|MakeVector (.x (Transformation|GetVelocity (Variables|Default|GetPlayerPawn))) (.y (Transformation|GetVelocity (Variables|Default|GetPlayerPawn))) (Math|Float|Absolute(Float) (.z (Transformation|GetVelocity (Variables|Default|GetPlayerPawn)))))) true true)))

(fn GetOverrideGravity
  (params (OverrideGravityValue :type Float%20%28double-precision%29))
  (return (select (Variables|Default|GetShouldCompenstateForGravity) (Components|Movement|GetGravityZ (Variables|/Script/Engine.Character|GetCharacterMovement (Variables|Default|GetPlayerPawn))) OverrideGravityValue)))

(fn GiveQuestCredit
  ;[1 1 1] Give Quest credit
  (if (Networking|IsServer)
    (Quest|SendComplexCustomStatEvent (Quests|GetQuestManager (Fortnite|GetFortPlayerControllerfromActor (Variables|Actor|GetInstigator)) "Invalid") 0 (Variables|Default|GetT_QuestTag) "(GameplayTags=)" 1)))

(fn HandleHits
  (params (Hit :type Hit%20Result%20Structure))
  (bind (_bblockinghit _binitialoverlap _time _distance _location _impactpoint _normal _impactnormal _physmat _hitactor _hitcomponent) (Collision|BreakHitResult Hit))
  (Variables|Default|SetHookStopped true)
  (Variables|Default|SetHitLocation _impactpoint)
  (bind _rope (Variables|Default|GetRope)
    (:then
      (Class|BGrappleKnifeRope|TightenRope _rope)
      (Variables|Default|SetImpactNormal _impactnormal)
      (Transformation|SetActorRotation (Math|Conversions|RotationFromXVector (* _impactnormal -1.0)))
      (Game|Feedback|ClientPlayForceFeedback (Fortnite|GetFortPlayerControllerfromActor (Variables|Default|GetPlayerPawn)) "/Game/Weapons/ForceFeedback/FF_Generic_Throw.FF_Generic_Throw" "None" false false false)
      (Gameplay|AuthorityAttachtoActor (Variables|Default|GetFortAttachableProjectile) Hit _hitactor (Variables|Default|GetOffset%20Distance%20from%20Physics%20Mesh) (Variables|Default|GetOffsetDistanceFromBone) 0.75)
      (Utilities|Casting|CastToFortPawn _hitactor
        (:then
          _)
        (:CastFailed
          (HandleSticks|HandleSticks _hitactor _hitcomponent)))
      (Variables|Default|SetStruckActor _hitactor)
      (return 0))
    (:else)))

(fn HandleLaunch
  (if (Math|Boolean|ANDBoolean (not (Variables|/Script/FortniteGame.FortPlayerPawn|GetbIsSkydiving (Variables|Default|GetPlayerPawn))) (not (Parachute|GetPendingSkydiveLaunch (Variables|Default|GetPlayerPawn))) (not (Parachute|IsSkydiving (Variables|Default|GetPlayerPawn))))
    (return false)
    (else
      (return true))))

(fn HandleMomentum
  (bind _struckactor (Variables|Default|GetStruckActor))
  (Utilities|IsValid _struckactor
    (:Is%20Valid
      (bind _asfort_player_pawn (Utilities|Casting|CastToFortPlayerPawn _struckactor)
        (:then
          (if (and (Vehicle|IsinVehicle _asfort_player_pawn) false)
            (Variables|Default|SetMomentum (Vehicle|GetVehicleVelocity (Vehicle|GetVehicleInterface _asfort_player_pawn)))
            (else
              _)))
        (:CastFailed
          _))
      (Variables|Default|SetMomentum (Transformation|GetVelocity (Variables|Default|GetStruckActor))))))

(fn HandleSticks
  (params (Object :type Object%20Reference) (Component :type Primitive%20Component%20Object%20Reference))
  (Utilities|FlowControl|Sequence
    (:then_0
      (HandleSticks|HandleStuckBuilding Object Component))
    (:then_1
      (HandleSticks|HandleStuckVehicle Object Component))
    (:then_2
      (HandleSticks|HandleStuckSupplyDrop Object Component))
    (:then_3
      (HandleSticks|HandleStuckPhysicsObject Object Component))))

(fn HandleStuckBuilding
  (params (Object :type Object%20Reference) (Component :type Primitive%20Component%20Object%20Reference))
  (bind _asbuilding_wall (Utilities|Casting|CastToBuildingWall Object))
  (LocalVariables|SetLocalWall _asbuilding_wall)
  (bind _asstatic_mesh_component (Utilities|Casting|CastToStaticMeshComponent Component))
  (bind _returnvalue (BuildingWall|IsDoorComponent (LocalVariables|GetLocalWall) _asstatic_mesh_component))
  (if _returnvalue
    (Transformation|AttachActorToComponent Component "None" "KeepWorld" "KeepWorld" "KeepWorld" false)))

(fn HandleStuckPhysicsObject
  (params (Object :type Object%20Reference) (Component :type Primitive%20Component%20Object%20Reference))
  (bind _asactor (Utilities|Casting|CastToActor Object))
  (Utilities|IsValid (Actor|GetComponentbyClass _asactor "/Script/FortniteGame.FortPhysicsObjectComponent")
    (:Is%20Valid
      (Transformation|AttachActorToComponent Component))))

(fn HandleStuckProjectiles
  (params (Object :type Object%20Reference) (Component :type Primitive%20Component%20Object%20Reference))
  (Utilities|FlowControl|Sequence
    (:then_0
      (HandleSticks|HandleStuckBuilding Object Component))
    (:then_1
      (HandleSticks|HandleStuckVehicle Object Component))
    (:then_2
      (HandleSticks|HandleStuckSupplyDrop Object Component))))

(fn HandleStuckSupplyDrop
  (params (Object :type Object%20Reference) (Component :type Primitive%20Component%20Object%20Reference))
  (Utilities|FlowControl|Sequence
    (:then_0
      (Utilities|Casting|CastToAthenaSupplyDrop Object)
      (Transformation|AttachActorToComponent Component "None" "KeepWorld" "KeepWorld" "KeepWorld"))
    (:then_1
      (Utilities|Casting|CastToAthenaSupplyDropBalloon Object)
      (Transformation|AttachActorToComponent Component "None" "KeepWorld" "KeepWorld" "KeepWorld" false))))

(fn HandleStuckVehicle
  (params (Object :type Object%20Reference) (Component :type Primitive%20Component%20Object%20Reference))
  (Utilities|Casting|CastToFortAthenaVehicle Object)
  (Transformation|AttachActorToComponent Component "None" "KeepWorld" "KeepWorld" "KeepWorld" false))

(fn IncapacitateTarget
  (bind _target (Variables|Default|GetTarget))
  (bind _returnvalue (Pawn|GetController _target))
  (Components|Movement|StopMovementImmediately (Variables|/Script/Engine.Character|GetCharacterMovement (Variables|Default|GetTarget)))
  (Input|SetIgnoreMoveInput _returnvalue true)
  ;[1 1 1] Makes target face instigator
  (Pawn|SetControlRotation _returnvalue (Math|Rotator|FindLookatRotation (Transformation|GetActorLocation (Variables|Default|GetTarget)) (Transformation|GetActorLocation (Variables|Default|GetPlayerPawn)))))

(fn MomentumCheck
  (params (A :type Vector) (A2 :type Vector))
  ;[1 1 1] Only add momentum if it's a net gain
  (return (select (> (Math|Vector|VectorLength (+ A2 (Variables|Default|GetMomentum))) (Math|Vector|VectorLength A)) (Variables|Default|GetMomentum) "0, 0, 0")))

(fn OnRep_HookStopped)

(fn OnRep_Momentum)

(fn OnRep_StruckActor)

(fn OnRep_Target)

(fn PickGrappleMontage
  (params (PitchAngle :type Float%20%28double-precision%29) (InHoverboard :type Boolean))
  (bind (_minpitch_6_c27dbdc64650f9ba5ebf09be2564561b _maxpitch_7_75538945434998adec50baa0edf1654a _sectiontoplay_12_2e20dbf74b43f4e8973e488482dca3f7) (Utilities|Struct|BreakToyAngleSelectorInfo (Utilities|Array|Get(aref) (Variables|Default|GetLaunchAngleVariations))))
  (LocalVariables|SetCleanedAngle (Math|Float|ClampAngle 0.0 -9e+01 9e+01))
  (for _array_element (Variables|Default|GetLaunchAngleVariations)
    (if (and (>= (LocalVariables|GetCleanedAngle) (Utilities|Struct|BreakToyAngleSelectorInfo _array_element)) (<= (LocalVariables|GetCleanedAngle) (.MaxPitch_7_75538945434998ADEC50BAA0EDF1654A (Utilities|Struct|BreakToyAngleSelectorInfo _array_element))))
      (return (.SectionToPlay_12_2E20DBF74B43F4E8973E488482DCA3F7 (Utilities|Struct|BreakToyAngleSelectorInfo _array_element)))))
  (return _sectiontoplay_12_2e20dbf74b43f4e8973e488482dca3f7))

(fn PickGrappleMontageSection
  (params (PitchAngle :type Float%20%28double-precision%29))
  (bind (_minpitch_6_c27dbdc64650f9ba5ebf09be2564561b _maxpitch_7_75538945434998adec50baa0edf1654a _sectiontoplay_12_2e20dbf74b43f4e8973e488482dca3f7) (Utilities|Struct|BreakToyAngleSelectorInfo (Utilities|Array|Get(aref) (Variables|Default|GetLaunchAngleVariations))))
  (LocalVariables|SetCleanedAngle (Math|Float|ClampAngle PitchAngle -9e+01 9e+01))
  (for _array_element (Variables|Default|GetLaunchAngleVariations)
    (if (and (>= (LocalVariables|GetCleanedAngle) (Utilities|Struct|BreakToyAngleSelectorInfo _array_element)) (<= (LocalVariables|GetCleanedAngle) (.MaxPitch_7_75538945434998ADEC50BAA0EDF1654A (Utilities|Struct|BreakToyAngleSelectorInfo _array_element))))
      (return (.SectionToPlay_12_2E20DBF74B43F4E8973E488482DCA3F7 (Utilities|Struct|BreakToyAngleSelectorInfo _array_element)))))
  (return _sectiontoplay_12_2e20dbf74b43f4e8973e488482dca3f7))

(fn PlayStickSounds
  (params (Location :type Vector) (HitActor :type Actor%20Object%20Reference))
  (Utilities|IsValid (Variables|Default|GetPlayerPawn)
    (:Is%20Valid
      (Utilities|IsValid HitActor
        (:Is%20Valid
          ;[1 1 1] Plays if an actor is hit
          (Utilities|FlowControl|Sequence
            (:then_0
              (if (Pawn|IsLocallyControlled (Variables|Default|GetPlayerPawn))
                _))
            (:then_1
              (if (Pawn|IsLocallyControlled (Utilities|Casting|CastToFortPlayerPawn HitActor))
                _
                (else
                  _)))))
        (:Is%20Not%20Valid
          (if (Pawn|IsLocallyControlled (Variables|Default|GetPlayerPawn))
            _
            (else
              _)))))
    (:Is%20Not%20Valid
      _)))

(fn SpawnRope
  (params (PlayerPawn :type Fort%20Player%20Pawn%20Object%20Reference))
  ;[1 1 1] Spawn The Rope
  (bind _returnvalue (Game|SpawnActorfromClass "/GrappleWeapon/Items/GrappleKnife/V2/B_GrappleKnife_Rope.B_GrappleKnife_Rope_C" (Math|Transform|MakeTransform (Transformation|GetActorLocation)) "Undefined" "MultiplyWithRoot" (Variables|Default|GetAs%20B%20Grapple%20Knife) PlayerPawn 0))
  (bind _output_get (Variables|Default|SetRope _returnvalue))
  (Class|FortHookGunRope|SetWeaponandProjectile _output_get (Variables|Default|GetAs%20B%20Grapple%20Knife) self)
  (return _output_get))

(fn ConstructionScript
  (bind _speed (Variables|Default|GetSpeed))
  (bind _returnvalue (ScalableFloat|GetValueatLevel _speed))
  (bind _capsulecomponent (Variables|Collision|GetCapsuleComponent))
  (bind _projectilemovementcomponent (Variables|Movement|GetProjectileMovementComponent))
  (bind _projecticleradius (Variables|Default|GetProjecticleRadius))
  (|CallParentFunction)
  (Variables|/Script/Engine.ProjectileMovementComponent|SetInitialSpeed _returnvalue _projectilemovementcomponent)
  (Variables|/Script/Engine.ProjectileMovementComponent|SetMaxSpeed _returnvalue _projectilemovementcomponent)
  (Actor|SetLifeSpan (ScalableFloat|GetValueatLevel (Variables|Default|GetBaseLifespan)))
  (Components|Capsule|SetCapsuleRadius _capsulecomponent _projecticleradius false)
  (Components|Capsule|SetCapsuleHalfHeight _capsulecomponent _projecticleradius false))
