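; Gameplay ability for the V2 grapple knife, named as a passive that blocks actions.
; Activation only prints a "BlockedActions" debug string and never ends the ability itself, so any blocking logic lies outside this graph.
; When the ability ends, cancelled or not, it sends the Event.GrappleKnife.Secondary.ResetRotation gameplay event to the activating pawn.
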(event Ability|EventActivateAbility
  (Development|PrintString "BlockedActions"))

(event Ability|EventOnEndAbility
  (params (bWasCancelled :type Boolean))
  (Ability|SendGameplayEventtoActor (Ability|GetActivatingPawn) "(TagName=\"Event.GrappleKnife.Secondary.ResetRotation\")"))
