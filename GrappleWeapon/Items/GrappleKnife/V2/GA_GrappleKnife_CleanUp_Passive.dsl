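; Passive Grapple Knife ability that handles movement and camera cleanup for the activating pawn.
; On landing, it removes the InitialFire gameplay effect if the pawn has that tag.
; Gameplay events drive a temporary rotation-rate override, reverted after a delay, and a grapple camera modifier, cleared by a three-second timer.
; Ending the ability restores rotate-to-camera and rotation rate, and disables the camera modifier.
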(event Custom|RemoveInitialFireGE
  (params (Hit :type Hit%20Result%20Structure%20%28by%20ref%29))
  (bind _returnvalue (Ability|GetActivatingPawn))
  (bind _returnvalue_2 (Ability|GetAbilitySystemComponent _returnvalue))
  (if (GameplayTags|HasMatchingGameplayTag _returnvalue_2 "(TagName=\"Granted.GrappleKnife.InitialFire\")")
    (GameplayEffects|RemoveActiveGameplayEffectbySourceEffect _returnvalue_2 (Variables|Default|GetGE_InitialFire) _returnvalue_2 -1)))

(event Ability|EventActivateAbility
  (bind _returnvalue (Ability|GetActivatingPawn))
  (bind _returnvalue_2 (Ability|GetActivatingPawn))
  (bind _returnvalue_3 (Fortnite|GetFortPlayerControllerfromActor _returnvalue))
  (bind _returnvalue_4 (Ability|GetActivatingPawn))
  (bind _charactermovement (Variables|/Script/Engine.Character|GetCharacterMovement _returnvalue_4))
  (EventDispatchers|BindEventtoOnPawnLanded (Ability|GetActivatingPawn) (delegate Custom|RemoveInitialFireGE))
  (Utilities|FlowControl|Sequence
    (:then_0
      (Ability|Async|WaitGameplayEventToActor _returnvalue_2 "(TagName=\"Event.GrappleKnife.Secondary.ResetRotation\")"
        (:EventReceived
          (Input|SetRotatePawntoCamera (Fortnite|GetFortPlayerControllerfromActor _returnvalue_4) true)
          (Variables|/Script/Engine.CharacterMovementComponent|SetRotationRate (Variables|Default|GetRotation%20Rate) _charactermovement)
          (Utilities|FlowControl|Delay (Variables|Default|GetDelayB4Reset))
          (Variables|/Script/Engine.CharacterMovementComponent|SetRotationRate "0.000000,-1.000000,0.000000" _charactermovement))))
    (:then_1
      (Ability|Async|WaitGameplayEventToActor _returnvalue_2 "(TagName=\"Event.GrappleKnife.Secondary.CameraModifier.Add\")"
        (:EventReceived
          (Utilities|IsValid (Variables|Default|GetCameraModifier)
            (:Is%20Valid
              (CallFunction|ClearCamera))
            (:Is%20Not%20Valid
              (Utilities|IsValid _returnvalue_3
                (:Is%20Valid
                  _))))
          (bind _returnvalue_5 (Fortnite|Camera|AddCameraModifierforOwningPlayerController "/GrappleWeapon/Items/Camera/B_GrappleKnife_CameraModifier.B_GrappleKnife_CameraModifier_C" (Variables|/Script/Engine.PlayerController|GetPlayerCameraManager _returnvalue_3)))
          (Variables|Default|SetCameraModifier _returnvalue_5)
          (Utilities|Time|SetTimerbyEvent (delegate Custom|DisableCameraModifier) 3.0))))))

(event Custom|DisableCameraModifier
  (CallFunction|ClearCamera))

(event Custom|ClearCamera
  (bind _returnvalue (Ability|GetActivatingPawn))
  (bind _cameramodifier (Variables|Default|GetCameraModifier))
  (Utilities|IsValid _returnvalue
    (:Is%20Valid
      (CameraModifier|DisableModifier _cameramodifier false)
      (Game|Player|RemoveCameraModifier (Variables|/Script/Engine.PlayerController|GetPlayerCameraManager (Game|Player|GetOwningPlayerController (Variables|/Script/Engine.PlayerController|GetPlayerCameraManager (Fortnite|GetFortPlayerControllerfromActor _returnvalue)))) _cameramodifier))))

(event Ability|EventOnEndAbility
  (params (bWasCancelled :type Boolean))
  (bind _returnvalue (Ability|GetActivatingPawn))
  (Input|SetRotatePawntoCamera (Fortnite|GetFortPlayerControllerfromActor _returnvalue) true)
  (Variables|/Script/Engine.CharacterMovementComponent|SetRotationRate "0.000000,-1.000000,0.000000" (Variables|/Script/Engine.Character|GetCharacterMovement _returnvalue))
  (CameraModifier|DisableModifier (Variables|Default|GetCameraModifier) true))
