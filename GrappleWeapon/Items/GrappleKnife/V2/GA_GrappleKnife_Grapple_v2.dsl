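; Grapple Knife secondary gameplay ability that fires a grapple hook projectile at a resolved target.
; The owning client finds the target and distance bucket, then RPCs it to the server, which commits cost and spawns the hook, homing on movable targets.
; During wind-up it ejects the player from vehicles, blocks grinding and primary attacks, and scales gravity.
; It ends when grapple movement starts or after a fail-safe delay.
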(event Ability|EventActivateAbility
  (bind _returnvalue (Ability|GetCurrentSourceWeapon))
  (bind _owningplayerpawn (Variables|Default|GetOwningPlayerPawn))
  (bind _returnvalue_2 (Vehicle|GetVehicleInterface _owningplayerpawn))
  (bind (_returnvalue_x _returnvalue_y _returnvalue_z) (Math|Vector|BreakVector (Vehicle|GetVehicleVelocity _returnvalue_2)))
  ;[0.0539 0.3079 0.3594] Activate ability
  (Variables|Default|SetbAbilityActive true)
  (bind _asfort_player_pawn (Utilities|Casting|CastToFortPlayerPawn (Ability|GetActivatingPawn)))
  (Variables|Default|SetOwningPlayerPawn _asfort_player_pawn)
  (bind _asb_grapple_knife (Utilities|Casting|CastToB_GrappleKnife (Ability|GetCurrentSourceWeapon)))
  (Variables|Default|SetAs%20B%20Grapple%20Knife _asb_grapple_knife)
  (CallFunction|ThrowGrapple)
  (Ability|HasAuthority
    (:True
      ;[1 1 1] The owning client resolves the target and RPCs it to the server.
      (if (and (Variables|Default|GetbAbilityActive) (Variables|Default|GetbServerHasTarget))
        (CallFunction|CallWeapononFiredEvent (stale Custom|ServerReceiveTarget DistanceBucket))
        ;[1 1 1] Eject Player in vehicles
        (if (Vehicle|IsinVehicle _owningplayerpawn)
          (Variables|Default|SetVehicleVelcity (Math|Vector|MakeVector _returnvalue_x _returnvalue_y (+ _returnvalue_z (Variables|Default|GetVehicleEjectZOffset))))
          (Vehicle|EjectPlayerPawn _returnvalue_2 _owningplayerpawn)
          (Pawn|Character|LaunchCharacterJump _owningplayerpawn (Variables|Default|GetVehicleVelcity) true true true true))
        ;[1 1 1] Block auto grinding on cast
        (if (not (GameplayTags|HasMatchingGameplayTag (Ability|GetAbilitySystemComponent (Ability|GetActivatingPawn)) "(TagName=\"MovementMode.Grinding\")"))
          (bind _returnvalue_3 (GameplayEffects|ApplyGameplayEffectToSelf :self (Ability|GetAbilitySystemComponent (Ability|GetActivatingPawn)) :GameplayEffectClass (Variables|Default|GetGE%20BlockGrinding) :Level 0.0))
          (Variables|Default|SetGrindingHandle _returnvalue_3))
        ;[1 1 1] Wind up delay
        (CallFunction|BlockPrimaryAttacks)
        (bind _returnvalue_4 (Utilities|Time|SetTimerbyEvent (delegate Custom|SetPawnRotation) (Variables|Default|GetStopRotationDelay)))
        (Variables|Default|SetSetPawnRotationHandle _returnvalue_4)
        (CallFunction|SetGravityScalarEnabled true)
        (CallFunction|CheckInitialFire)
        (CallFunction|DebugNoTargets)
        ;;[1 0 0] Only used for debugging
        (if (> (Variables|Default|GetProjectileSpawnTime) 0)
          (Utilities|FlowControl|Delay (Variables|Default|GetProjectileSpawnTime)))
        (Utilities|FlowControl|Sequence
          (:then_0
            ;[1 1 1] Spawn projectile and commit
            (Ability|HasAuthority
              (:True
                (Utilities|IsValid _returnvalue
                  (:Is%20Valid
                    (if (FortWeapon|Ammo|InventoryOwnerHasInfiniteUses _returnvalue)
                      (if (Variables|Costs|GetbApplyingCostsEnabled)
                        (Variables|Costs|SetbApplyingCostsEnabled false)
                        (bind _returnvalue_5 (Ability|CommitAbilityCost true)) ; Ensure we still broadcast commit events even without the cost
                        (if _returnvalue_5
                          (Variables|Costs|SetbApplyingCostsEnabled true)
                          (else
                            (Ability|EndAbility :!label _label)))
                        (else
                          (bind _returnvalue_6 (Ability|CommitAbilityCost true :!label _label_2))
                          (if (not _returnvalue_6)
                            (goto _label))))
                      (bind (_location _direction _speed _bhomingenabled) (Default|GetProjectileData))
                      (bind _returnvalue_7 (Projectile|SpawnProjectile :ProjectileClass (Variables|Ability|GetProjectileClass) :RequestedBy (Variables|/Script/FortniteGame.FortPawn|GetCurrentWeapon (Ability|GetActivatingPawn)) :SpawnLocation _location :SpawnRotation _direction :SpawnDirection _direction :EffectContainerSpecToApplyOnHit (Ability|MakeGameplayEffectContainerSpec (Variables|Default|GetHitApplicationTag)) :bUseDefaultPhysics false :InitialSpeed _speed :GravityScale 0.0 :ChargePercent 0.0))
                      (bind _asb_prj_grapple_hook (Utilities|Casting|CastToB_Prj_GrappleHook _returnvalue_7)
                        (:then
                          (if _bhomingenabled
                            ;[1 1 1] Lock on to movable targets
                            (CallFunction|SetLockedOnTarget (Variables|/Script/FortniteGame.FortProjectileBase|GetProjectileMovementComponent _asb_prj_grapple_hook) (Variables|HitInfo|GetHitActor) (Variables|HitInfo|GetHitComponent) (Variables|HitInfo|GetRelativeHitLocation))
                            (else
                              ;[1 1 1] Reset Lock on
                              (Projectile|SetHomingData (Variables|/Script/FortniteGame.FortProjectileBase|GetProjectileMovementComponent _asb_prj_grapple_hook) (Utilities|Struct|MakeProjectileHomingData "None" 0.0 0.0 0.5 0.0 7 0 0 "0.000000,0.000000,0.000000" "0.000000,0.000000,0.000000" "Auto" false true) :!class /Script/FortniteGame.FortProjectileMovementComponent))))
                        (:CastFailed
                          _))
                      ;[1 1 1] Wait to end until movement starts
                      (bind _asynctaskproxy (Ability|Tasks|WaitGameplayEvent "(TagName=\"Event.GrappleKnife.Secondary.MovementStarted\")" :!label _label_3)
                        (:then
                          (Variables|Default|SetEndOnMovementTask _asynctaskproxy))
                        (:EventReceived
                          (Ability|EndAbility))))
                    (goto _label_2))
                  (:Is%20Not%20Valid
                    _))
                (goto _label))
              (:False
                _))
            (goto _label_3))
          (:then_1
            ;[1 1 1] FF
            (ForceFeedback|SpawnForceFeedbackAttached "/Game/Weapons/ForceFeedback/FF_Generic_Melee.FF_Generic_Melee" (Variables|/Script/Engine.Actor|GetRootComponent (Ability|GetActivatingPawn)))
            ;[0.08779 1 0.6933] Custom cue for weapon fired
            (GameplayCue|ExecuteGameplayCueOnActor(Burst) (Ability|GetAvatarActorfromActorInfo) (Variables|Default|GetGC_WeaponFired) (Ability|GameplayCue|MakeGameplayCueParameters 0.0)))
          (:then_2
            ;[1 1 1] Fail Safe end
            (Utilities|FlowControl|Delay 1.0)
            (Ability|EndAbility)))))))

(event Custom|SetPawnRotation
  ;[1 1 1] Allow a delay to have player turn forward before stopping pawn rotation to camera
  (Input|SetRotatePawntoCamera (Fortnite|GetFortPlayerControllerfromActor (Ability|GetActivatingPawn)) false))

(event Custom|ServerReceiveTarget
  (params (WorldHitLocation :type Vector) (RelativeHitLocation :type Vector) (DistanceBucket :type Integer) (HitComponent :type Scene%20Component%20Object%20Reference))
  (bind _returnvalue (Ability|GetCurrentSourceWeapon))
  (bind _owningplayerpawn (Variables|Default|GetOwningPlayerPawn))
  (bind _returnvalue_2 (Vehicle|GetVehicleInterface _owningplayerpawn))
  (bind (_returnvalue_x _returnvalue_y _returnvalue_z) (Math|Vector|BreakVector (Vehicle|GetVehicleVelocity _returnvalue_2)))
  ;[1 1 1] The owning client resolves the target and RPCs it to the server.
  (Variables|HitInfo|SetWorldHitLocation WorldHitLocation)
  (Variables|HitInfo|SetRelativeHitLocation RelativeHitLocation)
  (Utilities|IsValid HitComponent
    (:Is%20Valid
      (bind _output_get (Variables|HitInfo|SetHitComponent HitComponent))
      (Variables|HitInfo|SetHitActor (Components|GetOwner _output_get)))
    (:Is%20Not%20Valid
      (Variables|HitInfo|SetHitComponent 0)
      (Variables|HitInfo|SetHitActor 0)))
  (Variables|Default|SetbServerHasTarget true)
  (if (and (Variables|Default|GetbAbilityActive) (Variables|Default|GetbServerHasTarget))
    (CallFunction|CallWeapononFiredEvent DistanceBucket)
    ;[1 1 1] Eject Player in vehicles
    (if (Vehicle|IsinVehicle _owningplayerpawn)
      (Variables|Default|SetVehicleVelcity (Math|Vector|MakeVector _returnvalue_x _returnvalue_y (+ _returnvalue_z (Variables|Default|GetVehicleEjectZOffset))))
      (Vehicle|EjectPlayerPawn _returnvalue_2 _owningplayerpawn)
      (Pawn|Character|LaunchCharacterJump _owningplayerpawn (Variables|Default|GetVehicleVelcity) true true true true))
    ;[1 1 1] Block auto grinding on cast
    (if (not (GameplayTags|HasMatchingGameplayTag (Ability|GetAbilitySystemComponent (Ability|GetActivatingPawn)) "(TagName=\"MovementMode.Grinding\")"))
      (bind _returnvalue_3 (GameplayEffects|ApplyGameplayEffectToSelf :self (Ability|GetAbilitySystemComponent (Ability|GetActivatingPawn)) :GameplayEffectClass (Variables|Default|GetGE%20BlockGrinding) :Level 0.0))
      (Variables|Default|SetGrindingHandle _returnvalue_3))
    ;[1 1 1] Wind up delay
    (CallFunction|BlockPrimaryAttacks)
    (bind _returnvalue_4 (Utilities|Time|SetTimerbyEvent (delegate Custom|SetPawnRotation) (Variables|Default|GetStopRotationDelay)))
    (Variables|Default|SetSetPawnRotationHandle _returnvalue_4)
    (CallFunction|SetGravityScalarEnabled true)
    (CallFunction|CheckInitialFire)
    (CallFunction|DebugNoTargets)
    ;;[1 0 0] Only used for debugging
    (if (> (Variables|Default|GetProjectileSpawnTime) 0)
      (Utilities|FlowControl|Delay (Variables|Default|GetProjectileSpawnTime)))
    (Utilities|FlowControl|Sequence
      (:then_0
        ;[1 1 1] Spawn projectile and commit
        (Ability|HasAuthority
          (:True
            (Utilities|IsValid _returnvalue
              (:Is%20Valid
                (if (FortWeapon|Ammo|InventoryOwnerHasInfiniteUses _returnvalue)
                  (if (Variables|Costs|GetbApplyingCostsEnabled)
                    (Variables|Costs|SetbApplyingCostsEnabled false)
                    (bind _returnvalue_5 (Ability|CommitAbilityCost true)) ; Ensure we still broadcast commit events even without the cost
                    (if _returnvalue_5
                      (Variables|Costs|SetbApplyingCostsEnabled true)
                      (else
                        (Ability|EndAbility :!label _label)))
                    (else
                      (bind _returnvalue_6 (Ability|CommitAbilityCost true :!label _label_2))
                      (if (not _returnvalue_6)
                        (goto _label))))
                  (bind (_location _direction _speed _bhomingenabled) (Default|GetProjectileData))
                  (bind _returnvalue_7 (Projectile|SpawnProjectile :ProjectileClass (Variables|Ability|GetProjectileClass) :RequestedBy (Variables|/Script/FortniteGame.FortPawn|GetCurrentWeapon (Ability|GetActivatingPawn)) :SpawnLocation _location :SpawnRotation _direction :SpawnDirection _direction :EffectContainerSpecToApplyOnHit (Ability|MakeGameplayEffectContainerSpec (Variables|Default|GetHitApplicationTag)) :bUseDefaultPhysics false :InitialSpeed _speed :GravityScale 0.0 :ChargePercent 0.0))
                  (bind _asb_prj_grapple_hook (Utilities|Casting|CastToB_Prj_GrappleHook _returnvalue_7)
                    (:then
                      (if _bhomingenabled
                        ;[1 1 1] Lock on to movable targets
                        (CallFunction|SetLockedOnTarget (Variables|/Script/FortniteGame.FortProjectileBase|GetProjectileMovementComponent _asb_prj_grapple_hook) (Variables|HitInfo|GetHitActor) (Variables|HitInfo|GetHitComponent) (Variables|HitInfo|GetRelativeHitLocation))
                        (else
                          ;[1 1 1] Reset Lock on
                          (Projectile|SetHomingData (Variables|/Script/FortniteGame.FortProjectileBase|GetProjectileMovementComponent _asb_prj_grapple_hook) (Utilities|Struct|MakeProjectileHomingData "None" 0.0 0.0 0.5 0.0 7 0 0 "0.000000,0.000000,0.000000" "0.000000,0.000000,0.000000" "Auto" false true) :!class /Script/FortniteGame.FortProjectileMovementComponent))))
                    (:CastFailed
                      _))
                  ;[1 1 1] Wait to end until movement starts
                  (bind _asynctaskproxy (Ability|Tasks|WaitGameplayEvent "(TagName=\"Event.GrappleKnife.Secondary.MovementStarted\")" :!label _label_3)
                    (:then
                      (Variables|Default|SetEndOnMovementTask _asynctaskproxy))
                    (:EventReceived
                      (Ability|EndAbility))))
                (goto _label_2))
              (:Is%20Not%20Valid
                _))
            (goto _label))
          (:False
            _))
        (goto _label_3))
      (:then_1
        ;[1 1 1] FF
        (ForceFeedback|SpawnForceFeedbackAttached "/Game/Weapons/ForceFeedback/FF_Generic_Melee.FF_Generic_Melee" (Variables|/Script/Engine.Actor|GetRootComponent (Ability|GetActivatingPawn)))
        ;[0.08779 1 0.6933] Custom cue for weapon fired
        (GameplayCue|ExecuteGameplayCueOnActor(Burst) (Ability|GetAvatarActorfromActorInfo) (Variables|Default|GetGC_WeaponFired) (Ability|GameplayCue|MakeGameplayCueParameters 0.0)))
      (:then_2
        ;[1 1 1] Fail Safe end
        (Utilities|FlowControl|Delay 1.0)
        (Ability|EndAbility)))))

(event Custom|Throw%20Grapple
  (bind _returnvalue (Ability|GetCurrentSourceWeapon))
  (bind _owningplayerpawn (Variables|Default|GetOwningPlayerPawn))
  (bind _returnvalue_2 (Vehicle|GetVehicleInterface _owningplayerpawn))
  (bind (_returnvalue_x _returnvalue_y _returnvalue_z) (Math|Vector|BreakVector (Vehicle|GetVehicleVelocity _returnvalue_2)))
  ;[1 1 1] The owning client resolves the target and RPCs it to the server.
  (Ability|HasAuthority
    (:False
      (bind (_hittarget _hitactor _hitcomponent _hitlocation _distancebucket) (CallFunction|GetTargetPointAndDistanceBucket))
      (if _hittarget
        (Variables|HitInfo|SetWorldHitLocation _hitlocation)
        (bind _output_get (Variables|HitInfo|SetRelativeHitLocation (CallFunction|CalculateRelativeLocation _hitcomponent _hitlocation)))
        (CallFunction|CallWeapononFiredEvent _distancebucket)
        (CallFunction|ServerReceiveTarget _hitlocation _output_get _distancebucket _hitcomponent)
        ;[1 1 1] Eject Player in vehicles
        (if (Vehicle|IsinVehicle _owningplayerpawn)
          (Variables|Default|SetVehicleVelcity (Math|Vector|MakeVector _returnvalue_x _returnvalue_y (+ _returnvalue_z (Variables|Default|GetVehicleEjectZOffset))))
          (Vehicle|EjectPlayerPawn _returnvalue_2 _owningplayerpawn)
          (Pawn|Character|LaunchCharacterJump _owningplayerpawn (Variables|Default|GetVehicleVelcity) true true true true))
        ;[1 1 1] Block auto grinding on cast
        (if (not (GameplayTags|HasMatchingGameplayTag (Ability|GetAbilitySystemComponent (Ability|GetActivatingPawn)) "(TagName=\"MovementMode.Grinding\")"))
          (bind _returnvalue_3 (GameplayEffects|ApplyGameplayEffectToSelf :self (Ability|GetAbilitySystemComponent (Ability|GetActivatingPawn)) :GameplayEffectClass (Variables|Default|GetGE%20BlockGrinding) :Level 0.0))
          (Variables|Default|SetGrindingHandle _returnvalue_3))
        ;[1 1 1] Wind up delay
        (CallFunction|BlockPrimaryAttacks)
        (bind _returnvalue_4 (Utilities|Time|SetTimerbyEvent (delegate Custom|SetPawnRotation) (Variables|Default|GetStopRotationDelay)))
        (Variables|Default|SetSetPawnRotationHandle _returnvalue_4)
        (CallFunction|SetGravityScalarEnabled true)
        (CallFunction|CheckInitialFire)
        (CallFunction|DebugNoTargets)
        ;;[1 0 0] Only used for debugging
        (if (> (Variables|Default|GetProjectileSpawnTime) 0)
          (Utilities|FlowControl|Delay (Variables|Default|GetProjectileSpawnTime)))
        (Utilities|FlowControl|Sequence
          (:then_0
            ;[1 1 1] Spawn projectile and commit
            (Ability|HasAuthority
              (:True
                (Utilities|IsValid _returnvalue
                  (:Is%20Valid
                    (if (FortWeapon|Ammo|InventoryOwnerHasInfiniteUses _returnvalue)
                      (if (Variables|Costs|GetbApplyingCostsEnabled)
                        (Variables|Costs|SetbApplyingCostsEnabled false)
                        (bind _returnvalue_5 (Ability|CommitAbilityCost true)) ; Ensure we still broadcast commit events even without the cost
                        (if _returnvalue_5
                          (Variables|Costs|SetbApplyingCostsEnabled true)
                          (else
                            (Ability|EndAbility :!label _label)))
                        (else
                          (bind _returnvalue_6 (Ability|CommitAbilityCost true :!label _label_2))
                          (if (not _returnvalue_6)
                            (goto _label))))
                      (bind (_location _direction _speed _bhomingenabled) (Default|GetProjectileData))
                      (bind _returnvalue_7 (Projectile|SpawnProjectile :ProjectileClass (Variables|Ability|GetProjectileClass) :RequestedBy (Variables|/Script/FortniteGame.FortPawn|GetCurrentWeapon (Ability|GetActivatingPawn)) :SpawnLocation _location :SpawnRotation _direction :SpawnDirection _direction :EffectContainerSpecToApplyOnHit (Ability|MakeGameplayEffectContainerSpec (Variables|Default|GetHitApplicationTag)) :bUseDefaultPhysics false :InitialSpeed _speed :GravityScale 0.0 :ChargePercent 0.0))
                      (bind _asb_prj_grapple_hook (Utilities|Casting|CastToB_Prj_GrappleHook _returnvalue_7)
                        (:then
                          (if _bhomingenabled
                            ;[1 1 1] Lock on to movable targets
                            (CallFunction|SetLockedOnTarget (Variables|/Script/FortniteGame.FortProjectileBase|GetProjectileMovementComponent _asb_prj_grapple_hook) (Variables|HitInfo|GetHitActor) (Variables|HitInfo|GetHitComponent) (Variables|HitInfo|GetRelativeHitLocation))
                            (else
                              ;[1 1 1] Reset Lock on
                              (Projectile|SetHomingData (Variables|/Script/FortniteGame.FortProjectileBase|GetProjectileMovementComponent _asb_prj_grapple_hook) (Utilities|Struct|MakeProjectileHomingData "None" 0.0 0.0 0.5 0.0 7 0 0 "0.000000,0.000000,0.000000" "0.000000,0.000000,0.000000" "Auto" false true) :!class /Script/FortniteGame.FortProjectileMovementComponent))))
                        (:CastFailed
                          _))
                      ;[1 1 1] Wait to end until movement starts
                      (bind _asynctaskproxy (Ability|Tasks|WaitGameplayEvent "(TagName=\"Event.GrappleKnife.Secondary.MovementStarted\")" :!label _label_3)
                        (:then
                          (Variables|Default|SetEndOnMovementTask _asynctaskproxy))
                        (:EventReceived
                          (Ability|EndAbility))))
                    (goto _label_2))
                  (:Is%20Not%20Valid
                    _))
                (goto _label))
              (:False
                _))
            (goto _label_3))
          (:then_1
            ;[1 1 1] FF
            (ForceFeedback|SpawnForceFeedbackAttached "/Game/Weapons/ForceFeedback/FF_Generic_Melee.FF_Generic_Melee" (Variables|/Script/Engine.Actor|GetRootComponent (Ability|GetActivatingPawn)))
            ;[0.08779 1 0.6933] Custom cue for weapon fired
            (GameplayCue|ExecuteGameplayCueOnActor(Burst) (Ability|GetAvatarActorfromActorInfo) (Variables|Default|GetGC_WeaponFired) (Ability|GameplayCue|MakeGameplayCueParameters 0.0)))
          (:then_2
            ;[1 1 1] Fail Safe end
            (Utilities|FlowControl|Delay 1.0)
            (Ability|EndAbility)))
        (else
          (Ability|EndAbility))))))

(event Custom|CheckInitialFire
  (bind _owningplayerpawn (Variables|Default|GetOwningPlayerPawn))
  (bind _returnvalue (Ability|GetAbilitySystemComponent _owningplayerpawn))
  ;[1 1 1] Check Initial Fire GE to apply || handles in air windup time
  (if (not (GameplayTags|HasMatchingGameplayTag _returnvalue (Variables|Default|GetTag%20to%20Check)))
    (bind _returnvalue_2 (GameplayEffects|ApplyGameplayEffectToSelf :self _returnvalue :GameplayEffectClass (Variables|Default|GetGE_InitialFire) :Level 0.0))
    (Variables|Default|SetInitialFireHandle _returnvalue_2)))

(event Ability|EventOnEndAbility
  (params (bWasCancelled :type Boolean))
  (bind _endonmovementtask (Variables|Default|GetEndOnMovementTask))
  ;[1 1 1] End ability
  (Variables|Default|SetbAbilityActive false)
  (Variables|Default|SetbServerHasTarget false)
  (Utilities|Time|ClearandInvalidateTimerbyHandle (Variables|Default|GetSetPawnRotationHandle))
  (Input|SetRotatePawntoCamera (Fortnite|GetFortPlayerControllerfromActor (Ability|GetActivatingPawn)) true)
  (CallFunction|SetGravityScalarEnabled false)
  (Ability|RemoveGameplayEffectFromOwnerWithHandle (Variables|Default|GetBlockPrimary_Handle))
  (Ability|RemoveGameplayEffectFromOwnerWithHandle (Variables|Default|GetGrindingHandle))
  (Utilities|IsValid _endonmovementtask
    (:Is%20Valid
      (GameplayTasks|EndTask _endonmovementtask))))

(fn BlockPrimaryAttacks
  (bind _returnvalue (GameplayEffects|ApplyGameplayEffectToSelf :self (Ability|GetAbilitySystemComponent (Ability|GetActivatingPawn)) :GameplayEffectClass (Variables|Default|GetGE%20BlockPrimary) :Level 0.0))
  (Variables|Default|SetBlockPrimary_Handle _returnvalue))

(fn CalculateRelativeLocation
  (params (SceneComponent :type Scene%20Component%20Object%20Reference) (WorldLocation :type Vector))
  (Utilities|IsValid SceneComponent
    (:Is%20Valid
      (Utilities|Casting|CastToFortPawn (Components|GetOwner SceneComponent)
        (:then
          _)
        (:CastFailed
          _))
      (return (Math|Transform|InverseTransformLocation (Transformation|GetWorldTransform SceneComponent) (CallFunction|ClampLocationinComponentBounds SceneComponent WorldLocation))))
    (:Is%20Not%20Valid
      (return "0, 0, 0"))))

(fn CalculateWorldLocation
  (params (SceneComponent :type Scene%20Component%20Object%20Reference) (RelativeLocation :type Vector))
  (Utilities|IsValid SceneComponent
    (:Is%20Valid
      (return (Math|Transform|TransformLocation (Transformation|GetWorldTransform SceneComponent) RelativeLocation)))
    (:Is%20Not%20Valid
      (return RelativeLocation))))

(fn CallWeaponOnFiredEvent
  (params (DistanceBucket :type Integer))
  (bind _owningplayerpawn (Variables|Default|GetOwningPlayerPawn))
  (bind _returnvalue (Ability|GetAbilitySystemComponent _owningplayerpawn))
  (bind _tag_to_check (Variables|Default|GetTag%20to%20Check))
  (bind _returnvalue_2 (GameplayTags|HasMatchingGameplayTag _returnvalue _tag_to_check))
  (bind _output_get (Variables|Default|SetWindUpTime (Math|Float|SelectFloat (Variables|Default|GetInAirWindUpDelay) (ScalableFloat|GetValueatLevel (Utilities|Array|Get(acopy) (Variables|Default|GetWindUpDelayBuckets) DistanceBucket)) _returnvalue_2)))
  (Variables|Default|SetProjectileSpawnTime (ScalableFloat|GetValueatLevel (Utilities|Array|Get(acopy) (Variables|Default|GetProjectileSpawnTimeBuckets) DistanceBucket)))
  (Development|PrintString (select _returnvalue_2 (Utilities|String|Append "IN AIR WIND UP DELAY:
" (Utilities|String|ToString(Float) _output_get)) (Utilities|String|Append "INITIAL WIND UP DELAY:
" (Utilities|String|ToString(Float) _output_get))))
  (Variables|Default|SetProjectileTravelTime (ScalableFloat|GetValueatLevel (Utilities|Array|Get(acopy) (Variables|Default|GetProjectileTravelTimeBuckets) DistanceBucket)))
  (Class|BGrappleKnife|EventonFired (Variables|Default|GetAs%20B%20Grapple%20Knife) DistanceBucket (Variables|Default|GetWindUpTime) (Variables|Default|GetProjectileTravelTime))
  (return))

(fn ClampLocationInComponentBounds
  (params (Component :type Scene%20Component%20Object%20Reference) (Location :type Vector%20%28by%20ref%29))
  (bind (_origin _boxextent _sphereradius) (Collision|GetComponentBounds Component))
  (return (+ (Math|Vector|ClampVectorSize (- Location _origin) 0.0 _sphereradius) _origin)))

(fn DebugNoTargets
  (bind _hitactor (Variables|HitInfo|GetHitActor))
  ;[1 1 1] Debug for no valid targets
  (if (Variables|/GrappleWeapon/Items/GrappleKnife/V2/B_GrappleKnife.B_GrappleKnife_C|GetNoTargetDebug (Variables|Default|GetAs%20B%20Grapple%20Knife))
    (Utilities|IsValid _hitactor
      (:Is%20Valid
        (Actor|SetLifeSpan _hitactor 0.1))
      (:Is%20Not%20Valid
        _)))
  (return))

(fn GetTargetPointandDistanceBucket
  (bind _owningplayerpawn (Variables|Default|GetOwningPlayerPawn))
  (bind (_bhittarget _targetactor _targetcomponent _targetpoint _hitdistance) (Default|GrappleKnifeGetTargetInfo (CallFunction|GetAvatarasFortPlayerPawn)))
  (if _bhittarget
    (bind _output_get (Variables|HitInfo|SetHitActor _targetactor))
    (bind _output_get_2 (Variables|HitInfo|SetHitComponent _targetcomponent))
    (return true _output_get _output_get_2 _targetpoint (CallFunction|GetDistanceBucket _hitdistance))
    (else
      (Variables|HitInfo|SetHitActor 0)
      (Variables|HitInfo|SetHitComponent 0)
      (FailedAction|UnabletoPerformAction _owningplayerpawn true (Variables|Default|GetFailedReasonOutOfRange))
      (if (Pawn|IsLocallyViewed _owningplayerpawn)
        (Audio|PlaySound2D (Variables|Default|GetCantDoThatSound)))
      (return false 0 0 "0, 0, 0" 0))))

(fn GetDistanceBucket
  (params (HitDistance :type Float%20%28double-precision%29))
  (bind (_array_element _array_index) (Utilities|Array|ForEachLoopwithBreak (Variables|Default|GetDistanceBuckets) :!label _label)
    (:LoopBody
      (if (>= HitDistance (ScalableFloat|GetValueatLevel _array_element))
        (LocalVariables|SetDistanceBucketRet _array_index)
        (else
          (goto _label Break))))
    (:Completed
      (return (LocalVariables|GetDistanceBucketRet)))))

(fn GetProjectileData
  (bind _safespawnlocation (Class|GrappleKnifeBPLib|GrappleKnifeFindSafeProjectileSpawn (CallFunction|GetAvatarasFortPlayerPawn) (+ (Math|Vector|RotateVectorAroundAxis (Math|Vector|SelectVector (LocalVariables|GetSpawnSwimmingOffset) (Math|Vector|SelectVector (LocalVariables|GetSpawnCrouchOffset) (LocalVariables|GetSpawnNormalOffset) (Variables|/Script/Engine.Character|GetbIsCrouched (CallFunction|GetAvatarasFortPlayerPawn))) (WaterBody|IsSurfaceSwimming (CallFunction|GetAvatarasFortPlayerPawn))) (.yaw (Transformation|GetActorRotation (CallFunction|GetAvatarasFortPlayerPawn))) "0.000000,0.000000,1.000000") (Transformation|GetActorLocation (CallFunction|GetAvatarasFortPlayerPawn)))))
  (return _safespawnlocation (Math|Conversions|RotationFromXVector (Math|Vector|Normalize (Math|Vector|ClampVectorSize (- (Math|Vector|SelectVector (CallFunction|CalculateWorldLocation (Variables|HitInfo|GetHitComponent) (Variables|HitInfo|GetRelativeHitLocation)) (Variables|HitInfo|GetWorldHitLocation) (CallFunction|ShouldUseRelativeLocation (Variables|HitInfo|GetHitComponent) (CallFunction|CalculateWorldLocation (Variables|HitInfo|GetHitComponent) (Variables|HitInfo|GetRelativeHitLocation)) (Variables|HitInfo|GetWorldHitLocation) (* (ScalableFloat|GetValueatLevel (Variables|Default|GetMaxTargetDistance)) 1.5))) _safespawnlocation) 0.0 (* (ScalableFloat|GetValueatLevel (Variables|Default|GetMaxTargetDistance)) 1.5)))) (Math|Float|SafeDivide (Math|Vector|VectorLength (Math|Vector|ClampVectorSize (- (Math|Vector|SelectVector (CallFunction|CalculateWorldLocation (Variables|HitInfo|GetHitComponent) (Variables|HitInfo|GetRelativeHitLocation)) (Variables|HitInfo|GetWorldHitLocation) (CallFunction|ShouldUseRelativeLocation (Variables|HitInfo|GetHitComponent) (CallFunction|CalculateWorldLocation (Variables|HitInfo|GetHitComponent) (Variables|HitInfo|GetRelativeHitLocation)) (Variables|HitInfo|GetWorldHitLocation) (* (ScalableFloat|GetValueatLevel (Variables|Default|GetMaxTargetDistance)) 1.5))) _safespawnlocation) 0.0 (* (ScalableFloat|GetValueatLevel (Variables|Default|GetMaxTargetDistance)) 1.5))) (Variables|/GrappleWeapon/Items/GrappleKnife/V2/B_GrappleKnife.B_GrappleKnife_C|GetProjectileTravelTime (Variables|Default|GetAs%20B%20Grapple%20Knife))) (and (CallFunction|ShouldUseRelativeLocation (Variables|HitInfo|GetHitComponent) (CallFunction|CalculateWorldLocation (Variables|HitInfo|GetHitComponent) (Variables|HitInfo|GetRelativeHitLocation)) (Variables|HitInfo|GetWorldHitLocation) (* (ScalableFloat|GetValueatLevel (Variables|Default|GetMaxTargetDistance)) 1.5)) (Class|GrappleKnifeBPLib|GrappleKnifeIsMovableActor (Variables|HitInfo|GetHitActor)))))

(fn IsHitActorMovable
  (bind _hitactor (Variables|HitInfo|GetHitActor))
  (Utilities|IsValid _hitactor
    (:Is%20Valid
      (return (or (Utilities|IsValid (Actor|GetComponentbyClass _hitactor "/Script/Engine.MovementComponent") :!class /Script/Engine.KismetSystemLibrary) (Utilities|IsValid (Actor|GetComponentbyClass _hitactor "/Script/FortniteGame.FortMovingPlatformComponent") :!class /Script/Engine.KismetSystemLibrary))))
    (:Is%20Not%20Valid
      (return false))))

(fn OverrideFailedReason
  (params (FailedReason :type Gameplay%20Tag%20Container%20Structure%20%28by%20ref%29))
  (bind _returnvalue (CallFunction|GetAvatarasFortPlayerPawn))
  (FailedAction|UnabletoPerformAction _returnvalue true (Variables|Default|GetFailedReasonOutOfAmmo))
  (Audio|PlaySound2D (Variables|Default|GetCantDoThatSound))
  (if (== FailedReason (GameplayTags|MakeLiteralGameplayTagContainer))
    (bind _asb_grapple_knife (Utilities|Casting|CastToB_GrappleKnife (Variables|/Script/FortniteGame.FortPawn|GetCurrentWeapon _returnvalue))
      (:then
        (Class|BGrappleKnife|NoAmmoFire _asb_grapple_knife))
      (:CastFailed
        _)))
  (return FailedReason))

(fn SetLockedonTarget
  (params (Projectile :type Fort%20Projectile%20Movement%20Component%20Object%20Reference) (LockedOnTarget :type Actor%20Object%20Reference) (LockedOnTargetComp :type Scene%20Component%20Object%20Reference) (TargetOffset :type Vector))
  (Projectile|SetHomingData Projectile (Utilities|Struct|MakeProjectileHomingData "LockOn" (ScalableFloat|GetValueatLevel (Variables|Default|GetRow_HomingMinTurnSpeed)) (ScalableFloat|GetValueatLevel (Variables|Default|GetRow_HomingMaxTurnSpeed)) (ScalableFloat|GetValueatLevel (Variables|Default|GetRow_HomingTurnRampTime)) (ScalableFloat|GetValueatLevel (Variables|Default|GetRow_HomingEffectTriggerRange)) 7 LockedOnTarget LockedOnTargetComp "0.000000,0.000000,0.000000" TargetOffset "Relative" false true) :!class /Script/FortniteGame.FortProjectileMovementComponent))

(fn SetGravityScalarEnabled
  (params (bEnabled :type Boolean))
  (bind (_velocity_x _velocity_y _velocity_z) (Math|Vector|BreakVector (Variables|/Script/Engine.MovementComponent|GetVelocity_X (Variables|/Script/Engine.Character|GetCharacterMovement (Variables|Default|GetOwningPlayerPawn)))))
  (Variables|/Script/Engine.CharacterMovementComponent|SetGravityScale (Math|Float|SelectFloat (ScalableFloat|GetValueatLevel (Variables|Default|GetGravityMultiplier)) 1.0 bEnabled) (Variables|/Script/Engine.Character|GetCharacterMovement (Variables|Default|GetOwningPlayerPawn))) ; Different value for gravity.\nThis lets us maintain momentum in XY\nbut fall slower.
  (if bEnabled
    (Variables|/Script/Engine.MovementComponent|SetVelocity_X (Math|Vector|MakeVector _velocity_x _velocity_y (* _velocity_z (Variables|Default|GetVelocityZOffset))) (Variables|/Script/Engine.Character|GetCharacterMovement (Variables|Default|GetOwningPlayerPawn)))))

(fn ShouldUseRelativeLocation
  (params (HitComponent :type Scene%20Component%20Object%20Reference) (CalculatedLocation :type Vector) (HitLocation :type Vector) (MaxDistance :type Float%20%28double-precision%29))
  (if (Utilities|IsValid HitComponent :!class /Script/Engine.KismetSystemLibrary)
    (return (<= (Math|Vector|DistanceSquared(Vector) CalculatedLocation HitLocation) (Math|Float|Square MaxDistance)))
    (else
      (return false))))
