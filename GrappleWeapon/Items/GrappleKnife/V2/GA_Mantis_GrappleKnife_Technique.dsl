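; Mantis technique ability for the GrappleKnife weapon that drives its attack state, with special handling for in-air attacks.
; It caches the knife on activation, calls StartAttack when a sequence starts and EndAttack when the ability ends.
; InAirAttack also calls ForceFieldinAirAtk, sends a primary-attack event and applies a tag-guarded effect against immediate reuse, avoiding stuck states.
; The custom source transform comes from the pawn's capsule.
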(event Ability|EventOnEndAbility
  (params (bWasCancelled :type Boolean))
  ;[1 1 1] End attack state
  (|CallParentFunction bWasCancelled)
  (Class|BGrappleKnife|EndAttack (Variables|Default|GetAs%20B%20Grapple%20Knife)))

(event Ability|EventActivateAbility
  ;[1 1 1] Cache reference to the grapple knife
  (|CallParentFunction)
  (bind _asb_grapple_knife (Utilities|Casting|CastToB_GrappleKnife (Ability|GetCurrentSourceWeapon)))
  (Variables|Default|SetAs%20B%20Grapple%20Knife _asb_grapple_knife))

(event Mantis|EventOnMantisAblilityTechniqueStarted
  (params (OwningTags :type Gameplay%20Tag%20Container%20Structure) (SequenceName :type Name :default "None") (StartsSequence :type Boolean) (EndsSequence :type Boolean))
  (bind _returnvalue (Ability|GetActivatingPawn))
  (bind _returnvalue_2 (== SequenceName "InAirAttack"))
  ;[1 1 1] Prevent in air attack from being used again for another second || prevents cases where player can get in a stuck state
  (|CallParentFunction OwningTags SequenceName StartsSequence EndsSequence)
  (if _returnvalue_2
    (Class|BGrappleKnife|ForceFieldinAirAtk (Variables|Default|GetAs%20B%20Grapple%20Knife))
    (Ability|SendGameplayEventtoActor _returnvalue "(TagName=\"Event.GrappleKnife.Primary.Attack\")")
    (if (not (GameplayTags|HasMatchingGameplayTag _returnvalue "(TagName=\"Granted.GrappleKnife.InAirFire\")"))
      (bind _returnvalue_3 (GameplayEffects|ApplyGameplayEffectToSelf :self (Ability|GetAbilitySystemComponent _returnvalue) :GameplayEffectClass (Variables|Default|GetGE_InAirFire) :Level 0.0))
      (Variables|Default|SetInAirFireHandle _returnvalue_3))
    (else
      _))
  ;;[1 1 1] Start attack state
  (if StartsSequence
    (Class|BGrappleKnife|StartAttack (Variables|Default|GetAs%20B%20Grapple%20Knife) _returnvalue_2)))

(fn GetCustomAbilitySourceTransform
  (return (Transformation|GetWorldTransform (Variables|/Script/Engine.Character|GetCapsuleComponent (Ability|GetActivatingPawn)))))
