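; Blueprint function library for the V2 Grapple Knife, providing targeting and projectile helpers plus its ledge config.
; Target info comes from GrappleWeaponLibrary using a locally held targeting config.
; The movable-actor check looks for movement, moving-platform or Dagwood network physics components, optionally recursing up attach parents.
; Safe spawn sphere-traces from the pawn to the desired point using the hook projectile's default radius, offsetting blocked spawns along the hit normal.
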(fn GrappleKnifeGetTargetInfo
  (params (SourcePlayerPawn :type Fort%20Player%20Pawn%20Object%20Reference) (DrawDebugDuration :type Float%20%28double-precision%29))
  (bind _returnvalue (Class|GrappleWeaponLibrary|GetGrappleWeaponTargetInfo SourcePlayerPawn (LocalVariables|GetTargetingConfig) DrawDebugDuration))
  (return (Utilities|Struct|BreakGrappleWeaponTargetInfo _returnvalue) (.TargetActor (Utilities|Struct|BreakGrappleWeaponTargetInfo _returnvalue)) (.TargetComponent (Utilities|Struct|BreakGrappleWeaponTargetInfo _returnvalue)) (.TargetPoint (Utilities|Struct|BreakGrappleWeaponTargetInfo _returnvalue)) (.HitDistance (Utilities|Struct|BreakGrappleWeaponTargetInfo _returnvalue)) false "0, 0, 0"))

(fn GrappleKnifeIsMovableActor
  (params (Actor :type Actor%20Object%20Reference) (bCheckAttachParent :type Boolean :default true))
  (Utilities|IsValid Actor
    (:Is%20Valid
      (if (Math|Boolean|ORBoolean (Utilities|IsValid (Actor|GetComponentbyClass Actor "/Script/Engine.MovementComponent") :!class /Script/Engine.KismetSystemLibrary) (Utilities|IsValid (Actor|GetComponentbyClass Actor "/Script/FortniteGame.FortMovingPlatformComponent") :!class /Script/Engine.KismetSystemLibrary) (Utilities|IsValid (Actor|GetComponentbyClass Actor "/Script/ValetRuntime.DagwoodNetworkPhysicsComponent") :!class /Script/Engine.KismetSystemLibrary))
        (return true)
        (elif (Variables|GetbCheckAttachParent)
          (return (CallFunction|GrappleKnifeIsMovableActor (Actor|GetAttachParentActor (Variables|GetActor)) true (Variables|GetActor)))
          (else
            (return false)))))
    (:Is%20Not%20Valid
      (return false))))

(fn GrappleKnifeFindSafeProjectileSpawn
  (params (SourcePlayerPawn :type Fort%20Player%20Pawn%20Object%20Reference) (DesiredSpawnLocation :type Vector))
  (bind _radius (Internal|GrappleKnifeGetProjectileRadius 0))
  (bind (_outhit _returnvalue) (Collision|SphereTraceByProfile SourcePlayerPawn (Math|Vector|MakeVector (.x (Transformation|GetActorLocation SourcePlayerPawn)) (.y (Transformation|GetActorLocation SourcePlayerPawn)) (.z DesiredSpawnLocation)) DesiredSpawnLocation _radius "FortProjectileHitAllPawns" false (Utilities|Array|MakeArray (Variables|GetSourcePlayerPawn)) "None" true "(R=1.000000,G=0.000000,B=0.000000,A=1.000000)" "(R=0.000000,G=1.000000,B=0.000000,A=1.000000)" 15.0))
  (if _returnvalue
    (return (+ (.Location (Collision|BreakHitResult _outhit)) (* (.Normal (Collision|BreakHitResult _outhit)) _radius)))
    (else
      (return (Variables|GetDesiredSpawnLocation)))))

(fn GrappleKnifeGetLedgeConfig
  (return (LocalVariables|GetLedgeConfig)))

(fn GrappleKnifeGetProjectileRadius
  (bind (_target _playerpawn _hookstopped _hitlocation _aircontrol _distancefrominstigator _sound_water _launchanglevariations _fireaimpitch _ropeactivetag _blockincomingpickups _impactnormal _offset_distance_from_physics_mesh _offsetdistancefrombone _struckactor _momentum _launchpad _vehiclevelocity _binhoverboardormotorcycle _lifespanonhit _tc_vehicleignorepull _ge_ropeactive _event_returned _event_grapplehit _t_questtag _fx_waterimpact _fx_reattach _row_standardvehiclearc _row_smallvehiclearc _row_vehicleforcemultiplier _row_jackleforcemultiplier _speed _baselifespan _shouldcompenstateforgravity _pulltopawnwhileinvehicle _as_b_grapple_knife _rope _cameramodifier _hit_actor _gc_grappled _smallvehiclestag _newvar _delaybeforeretract _projecticleradius) (Class|GetClassDefaults "/GrappleWeapon/Items/GrappleKnife/V2/B_Prj_GrappleHook.B_Prj_GrappleHook_C"))
  (return _projecticleradius))
