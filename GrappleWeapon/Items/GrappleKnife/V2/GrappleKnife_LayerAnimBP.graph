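# Item animation layer Blueprint for the Grapple Knife, overriding the player's final pose through the HasFired state machine.
# States cover ground, air, upper-body and vehicle casts, zip, air attack, wall bounce, release and fall, returning to the cached incoming pose.
# Nested conduit machines choose short, mid or long casts by distance bucket, using aim-pitch blendspaces.
# Inertialized transitions feed Dead Blending; Mantis technique events drive the air-attack flag.
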AnimBlueprint /GrappleWeapon/Items/GrappleKnife/V2/GrappleKnife_LayerAnimBP
target_skeleton=&Fortnite_M_Avg_Player_Skeleton
graphs [
 AnimGraph {nodes [
   Blend Poses by bool(BlendPose_1=Local Space Ref Pose(), bActiveValue=Get bIsFired() -> (bIsFired:bool)) {blendType=Linear}
   Output Pose() {layerGroup=None}
  ]}
 EventGraph {nodes [
   n1 = Event OnMantisTechniqueStarted!() -> (OutputDelegate, then, TechniqueIndex:int=0, TechniqueName:Name=None) {is=title K2Node_Event}
   > Set IsInAirAttack(IsInAirAttack:bool="Equal (Name)"(A=@n1.TechniqueName, B=InAirAttack)) -> (Output_Get:bool)
   > PrintString!(execute, InString="To String (Name)"(InName=@n1.TechniqueName), bPrintToScreen, bPrintToLog, TextColor, Duration, Key) -> (then)
   n2 = Event OnMantisTechniqueEnded!() -> (OutputDelegate, then, TechniqueIndex:int=0, TechniqueName:Name=None) {is=title K2Node_Event}
   > Set IsInAirAttack(IsInAirAttack=Select(Option 0:bool=false, Option 1:bool=false, Index="Equal (Name)"(A=@n2.TechniqueName, B=InAirAttack)) -> (ReturnValue:bool))
   Event BlueprintUpdateAnimation()
   Event BlueprintInitializeAnimation()
  ]}
 ItemAimOffset {nodes [
   Input Pose!() -> (Pose, AimOffsetAlpha:double, Yaw:double, Pitch:double) {name=InPose_AimOffset}
   Input Pose!() -> (Pose) {name=InPose_PreAimOffsets}
   Output Pose() {layerGroup=Item Override}
  ]}
 ItemFallAdditive {nodes [
   Input Pose() {name=InPoseItemFallAdditive}
   Output Pose() {layerGroup=Item Override}
  ]}
 ItemFallLandAdditive {nodes [
   Input Pose() {name=InPoseItemFallLandAdditive}
   Output Pose() {layerGroup=Item Override}
  ]}
 ItemFallLevel2Override {nodes [
   Input Pose() {name=InPose_FallLevel2}
   Output Pose() {layerGroup=Item Override}
  ]}
 ItemFinalPoseOverride {nodes [
   AnimGraphNode_StateMachine_0 = HasFired() {is=title AnimGraphNode_StateMachine}
   Output Pose(Result=Dead Blending!(Source:Pose Link Structure=@AnimGraphNode_StateMachine_0) -> (Pose:Pose Link Structure) {is=title AnimGraphNode_DeadBlending, bLinearlyInterpolateScales=true, defaultBlendMode=Linear}) {layerGroup=Item Override}
   Save cached pose 'SavedFinalPose'(Pose=Input Pose() {name=InFinalPose}) {cacheName=SavedFinalPose}
  ]}
 HasFired {
  entry=FullBodyState
  path=ItemFinalPoseOverride.AnimGraphNode_StateMachine_0.HasFired
  ids=id
  nodes [
   ToGround() {is=title AnimStateAliasNode, also {aliases [Fall, Release, ReleaseLoop, WallBounce, ZipGroundStart]}}
   ToAirFire() {is=title AnimStateAliasNode, also {aliases [Fall, Release, ReleaseLoop, WallBounce, ZipGroundStart]}}
   ToWallBounce() {is=title AnimStateAliasNode, also {aliases [Fall, Release, ReleaseLoop, ZipAirStart, ZipGroundStart, ZipUpperBody]}}
   ToZipAirAttack() {is=title AnimStateAliasNode, also {aliases [CastInAir, Fall, FullBodyState, Release, ReleaseLoop, WallBounce, ZipAirStart, ZipGroundStart]}}
   ExitUnexpectedly() {
    is=title AnimStateAliasNode
    also {aliases [
      CastFromVehicle
      CastGroundIdle
      CastGroundLoco
      CastInAir
      CastUpperBody
      Fall
      Release
      ReleaseLoop
      WallBounce
      ZipAirAttack
      ZipAirStart
      ZipGroundStart
      ZipUpperBody
     ]}
   }
   AnimStateNode_0 = CastFromVehicle() {is=title AnimStateNode}
   AnimStateNode_1 = FullBodyState() {is=title AnimStateNode}
   AnimStateNode_10 = ZipAirAttack() {is=title AnimStateNode}
   AnimStateNode_11 = ZipUpperBody() {is=title AnimStateNode}
   AnimStateNode_12 = ReleaseLoop() {is=title AnimStateNode}
   AnimStateNode_14 = CastGroundLoco() {is=title AnimStateNode}
   AnimStateNode_2 = CastGroundIdle() {is=title AnimStateNode}
   AnimStateNode_3 = CastUpperBody() {is=title AnimStateNode}
   AnimStateNode_4 = Fall() {is=title AnimStateNode}
   AnimStateNode_5 = Release() {is=title AnimStateNode}
   AnimStateNode_6 = CastInAir() {is=title AnimStateNode}
   AnimStateNode_7 = ZipAirStart() {is=title AnimStateNode}
   AnimStateNode_8 = ZipGroundStart() {is=title AnimStateNode}
   AnimStateNode_9 = WallBounce() {is=title AnimStateNode}
   AnimStateTransitionNode_0 = AnimStateTransitionNode() {automaticRuleTriggerTime=0, crossfadeDuration=0.1, logicType=TLT_Inertialization, priorityOrder=2}
   AnimStateTransitionNode_1 = AnimStateTransitionNode() {priorityOrder=4}
   AnimStateTransitionNode_10 = AnimStateTransitionNode() {automaticRuleTriggerTime=0, bAutomaticRuleBasedOnSequencePlayerInState=true, crossfadeDuration=0.5, logicType=TLT_Inertialization}
   AnimStateTransitionNode_11 = AnimStateTransitionNode() {crossfadeDuration=0.15, logicType=TLT_Inertialization}
   AnimStateTransitionNode_12 = AnimStateTransitionNode() {automaticRuleTriggerTime=0, bAutomaticRuleBasedOnSequencePlayerInState=true, logicType=TLT_Inertialization}
   AnimStateTransitionNode_13 = AnimStateTransitionNode() {logicType=TLT_Inertialization}
   AnimStateTransitionNode_14 = AnimStateTransitionNode() {crossfadeDuration=0.5, logicType=TLT_Inertialization}
   AnimStateTransitionNode_15 = AnimStateTransitionNode() {automaticRuleTriggerTime=0, bAutomaticRuleBasedOnSequencePlayerInState=true, crossfadeDuration=0.3, logicType=TLT_Inertialization}
   AnimStateTransitionNode_16 = AnimStateTransitionNode() {logicType=TLT_Inertialization, priorityOrder=53}
   AnimStateTransitionNode_17 = AnimStateTransitionNode() {logicType=TLT_Inertialization, priorityOrder=54}
   AnimStateTransitionNode_18 = AnimStateTransitionNode() {automaticRuleTriggerTime=0, bAutomaticRuleBasedOnSequencePlayerInState=true, logicType=TLT_Inertialization}
   AnimStateTransitionNode_19 = AnimStateTransitionNode() {automaticRuleTriggerTime=0, bAutomaticRuleBasedOnSequencePlayerInState=true, logicType=TLT_Inertialization}
   AnimStateTransitionNode_2 = AnimStateTransitionNode() {logicType=TLT_Inertialization, priorityOrder=3}
   AnimStateTransitionNode_20 = AnimStateTransitionNode() {automaticRuleTriggerTime=0, bAutomaticRuleBasedOnSequencePlayerInState=true, crossfadeDuration=0.5, logicType=TLT_Inertialization}
   AnimStateTransitionNode_21 = AnimStateTransitionNode() {automaticRuleTriggerTime=0, bAutomaticRuleBasedOnSequencePlayerInState=true, crossfadeDuration=0.7, logicType=TLT_Inertialization}
   AnimStateTransitionNode_22 = AnimStateTransitionNode() {automaticRuleTriggerTime=0, bAutomaticRuleBasedOnSequencePlayerInState=true, crossfadeDuration=0.4, logicType=TLT_Inertialization}
   AnimStateTransitionNode_23 = AnimStateTransitionNode() {automaticRuleTriggerTime=0, crossfadeDuration=0.4, logicType=TLT_Inertialization}
   AnimStateTransitionNode_3 = AnimStateTransitionNode() {crossfadeDuration=0.35, logicType=TLT_Inertialization, priorityOrder=2}
   AnimStateTransitionNode_4 = AnimStateTransitionNode() {automaticRuleTriggerTime=0, bAutomaticRuleBasedOnSequencePlayerInState=true, logicType=TLT_Inertialization}
   AnimStateTransitionNode_5 = AnimStateTransitionNode() {crossfadeDuration=0.1, logicType=TLT_Inertialization, priorityOrder=51}
   AnimStateTransitionNode_6 = AnimStateTransitionNode() {logicType=TLT_Inertialization, priorityOrder=100}
   AnimStateTransitionNode_7 = AnimStateTransitionNode() {logicType=TLT_Inertialization, priorityOrder=50}
   AnimStateTransitionNode_8 = AnimStateTransitionNode() {automaticRuleTriggerTime=0, bAutomaticRuleBasedOnSequencePlayerInState=true, logicType=TLT_Inertialization}
   AnimStateTransitionNode_9 = AnimStateTransitionNode() {logicType=TLT_Inertialization, priorityOrder=52}
  ]
  links [
   {
    dst=ZipGroundStart
    rule=ItemFinalPoseOverride.AnimGraphNode_StateMachine_0.HasFired.AnimStateTransitionNode_4.Transition
    src=CastFromVehicle
    via=AnimStateTransitionNode_4
   }
   {
    dst=CastGroundLoco
    rule=ItemFinalPoseOverride.AnimGraphNode_StateMachine_0.HasFired.AnimStateTransitionNode_11.Transition
    src=CastGroundIdle
    via=AnimStateTransitionNode_11
   }
   {
    dst=ZipGroundStart
    rule=ItemFinalPoseOverride.AnimGraphNode_StateMachine_0.HasFired.AnimStateTransitionNode_8.Transition
    src=CastGroundIdle
    via=AnimStateTransitionNode_8
   }
   {
    dst=CastGroundIdle
    rule=ItemFinalPoseOverride.AnimGraphNode_StateMachine_0.HasFired.AnimStateTransitionNode_3.Transition
    src=CastGroundLoco
    via=AnimStateTransitionNode_3
   }
   {
    dst=ZipGroundStart
    rule=ItemFinalPoseOverride.AnimGraphNode_StateMachine_0.HasFired.AnimStateTransitionNode_18.Transition
    src=CastGroundLoco
    via=AnimStateTransitionNode_18
   }
   {
    dst=ZipAirStart
    rule=ItemFinalPoseOverride.AnimGraphNode_StateMachine_0.HasFired.AnimStateTransitionNode_12.Transition
    src=CastInAir
    via=AnimStateTransitionNode_12
   }
   {
    dst=ZipUpperBody
    rule=ItemFinalPoseOverride.AnimGraphNode_StateMachine_0.HasFired.AnimStateTransitionNode_19.Transition
    src=CastUpperBody
    via=AnimStateTransitionNode_19
   }
   {
    dst=FullBodyState
    rule=ItemFinalPoseOverride.AnimGraphNode_StateMachine_0.HasFired.AnimStateTransitionNode_17.Transition
    src=ExitUnexpectedly
    via=AnimStateTransitionNode_17
   }
   {
    dst=CastFromVehicle
    rule=ItemFinalPoseOverride.AnimGraphNode_StateMachine_0.HasFired.AnimStateTransitionNode_1.Transition
    src=FullBodyState
    via=AnimStateTransitionNode_1
   }
   {
    dst=CastGroundIdle
    rule=ItemFinalPoseOverride.AnimGraphNode_StateMachine_0.HasFired.AnimStateTransitionNode_0.Transition
    src=FullBodyState
    via=AnimStateTransitionNode_0
   }
   {
    dst=CastGroundLoco
    rule=ItemFinalPoseOverride.AnimGraphNode_StateMachine_0.HasFired.AnimStateTransitionNode_2.Transition
    src=FullBodyState
    via=AnimStateTransitionNode_2
   }
   {
    dst=CastInAir
    rule=ItemFinalPoseOverride.AnimGraphNode_StateMachine_0.HasFired.AnimStateTransitionNode_6.Transition
    src=FullBodyState
    via=AnimStateTransitionNode_6
   }
   {
    dst=CastUpperBody
    rule=ItemFinalPoseOverride.AnimGraphNode_StateMachine_0.HasFired.AnimStateTransitionNode_13.Transition
    src=FullBodyState
    via=AnimStateTransitionNode_13
   }
   {
    dst=ReleaseLoop
    rule=ItemFinalPoseOverride.AnimGraphNode_StateMachine_0.HasFired.AnimStateTransitionNode_15.Transition
    src=Release
    via=AnimStateTransitionNode_15
   }
   {
    dst=Fall
    rule=ItemFinalPoseOverride.AnimGraphNode_StateMachine_0.HasFired.AnimStateTransitionNode_21.Transition
    src=ReleaseLoop
    via=AnimStateTransitionNode_21
   }
   {
    dst=CastInAir
    rule=ItemFinalPoseOverride.AnimGraphNode_StateMachine_0.HasFired.AnimStateTransitionNode_5.Transition
    src=ToAirFire
    via=AnimStateTransitionNode_5
   }
   {
    dst=FullBodyState
    rule=ItemFinalPoseOverride.AnimGraphNode_StateMachine_0.HasFired.AnimStateTransitionNode_7.Transition
    src=ToGround
    via=AnimStateTransitionNode_7
   }
   {
    dst=WallBounce
    rule=ItemFinalPoseOverride.AnimGraphNode_StateMachine_0.HasFired.AnimStateTransitionNode_16.Transition
    src=ToWallBounce
    via=AnimStateTransitionNode_16
   }
   {
    dst=ZipAirAttack
    rule=ItemFinalPoseOverride.AnimGraphNode_StateMachine_0.HasFired.AnimStateTransitionNode_9.Transition
    src=ToZipAirAttack
    via=AnimStateTransitionNode_9
   }
   {
    dst=ReleaseLoop
    rule=ItemFinalPoseOverride.AnimGraphNode_StateMachine_0.HasFired.AnimStateTransitionNode_23.Transition
    src=WallBounce
    via=AnimStateTransitionNode_23
   }
   {
    dst=ReleaseLoop
    rule=ItemFinalPoseOverride.AnimGraphNode_StateMachine_0.HasFired.AnimStateTransitionNode_22.Transition
    src=ZipAirAttack
    via=AnimStateTransitionNode_22
   }
   {
    dst=Release
    rule=ItemFinalPoseOverride.AnimGraphNode_StateMachine_0.HasFired.AnimStateTransitionNode_14.Transition
    src=ZipAirStart
    via=AnimStateTransitionNode_14
   }
   {
    dst=Release
    rule=ItemFinalPoseOverride.AnimGraphNode_StateMachine_0.HasFired.AnimStateTransitionNode_10.Transition
    src=ZipGroundStart
    via=AnimStateTransitionNode_10
   }
   {
    dst=Release
    rule=ItemFinalPoseOverride.AnimGraphNode_StateMachine_0.HasFired.AnimStateTransitionNode_20.Transition
    src=ZipUpperBody
    via=AnimStateTransitionNode_20
   }
  ]
 }
 CastFromVehicle {
  path=ItemFinalPoseOverride.AnimGraphNode_StateMachine_0.HasFired.AnimStateNode_0.CastFromVehicle
  nodes [
   AnimGraphNode_StateMachine_1 = CastUpperBody() {is=title AnimGraphNode_StateMachine, bAllowConduitEntryStates=true}
   Output Animation Pose(Result=Modify Curve(SourcePose=@AnimGraphNode_StateMachine_1, CurveValues_0=1, Alpha=1) {curveNames [BothLegFK], curveValues [0]})
  ]
 }
 CastUpperBody {
  entry=Conduit
  path=ItemFinalPoseOverride.AnimGraphNode_StateMachine_0.HasFired.AnimStateNode_0.CastFromVehicle.AnimGraphNode_StateMachine_1.CastUpperBody
  ids=id
  nodes [
   AnimStateConduitNode_0 = Conduit()
   AnimStateNode_1 = Mid() {is=title AnimStateNode}
   AnimStateNode_2 = Long() {is=title AnimStateNode}
   AnimStateNode_3 = Short() {is=title AnimStateNode}
   AnimStateTransitionNode_2 = AnimStateTransitionNode() {crossfadeDuration=0.15, logicType=TLT_Inertialization, priorityOrder=2}
   AnimStateTransitionNode_3 = AnimStateTransitionNode() {crossfadeDuration=0.15, logicType=TLT_Inertialization, priorityOrder=100}
   AnimStateTransitionNode_4 = AnimStateTransitionNode() {logicType=TLT_Inertialization}
  ]
  links [
   {
    dst=Long
    rule=ItemFinalPoseOverride.AnimGraphNode_StateMachine_0.HasFired.AnimStateNode_0.CastFromVehicle.AnimGraphNode_StateMachine_1.CastUpperBody.AnimStateTransitionNode_2.Transition
    src=Conduit
    via=AnimStateTransitionNode_2
   }
   {
    dst=Mid
    rule=ItemFinalPoseOverride.AnimGraphNode_StateMachine_0.HasFired.AnimStateNode_0.CastFromVehicle.AnimGraphNode_StateMachine_1.CastUpperBody.AnimStateTransitionNode_3.Transition
    src=Conduit
    via=AnimStateTransitionNode_3
   }
   {
    dst=Short
    rule=ItemFinalPoseOverride.AnimGraphNode_StateMachine_0.HasFired.AnimStateNode_0.CastFromVehicle.AnimGraphNode_StateMachine_1.CastUpperBody.AnimStateTransitionNode_4.Transition
    src=Conduit
    via=AnimStateTransitionNode_4
   }
  ]
 }
 Conduit {
  path=ItemFinalPoseOverride.AnimGraphNode_StateMachine_0.HasFired.AnimStateNode_0.CastFromVehicle.AnimGraphNode_StateMachine_1.CastUpperBody.AnimStateConduitNode_0.Conduit
  nodes [Result(bCanEnterTransition=true)]
 }
 Mid {
  path=ItemFinalPoseOverride.AnimGraphNode_StateMachine_0.HasFired.AnimStateNode_0.CastFromVehicle.AnimGraphNode_StateMachine_1.CastUpperBody.AnimStateNode_1.Mid
  nodes [Output Animation Pose(Result=Blendspace Player 'GrappleKnife_Cast_Loco_BS'!(X=Get AimPitch()) -> (Pose) {bLoop=false, blendSpace=/GrappleWeapon/Items/Animations/GrappleKnife_Cast_Loco_BS, groupName=Cast, method=SyncGroup})]
 }
 Long {
  path=ItemFinalPoseOverride.AnimGraphNode_StateMachine_0.HasFired.AnimStateNode_0.CastFromVehicle.AnimGraphNode_StateMachine_1.CastUpperBody.AnimStateNode_2.Long
  nodes [
   n1 = Blendspace Player 'GrappleKnife_Cast_Long_Loco_BS'!(X=Get AimPitch()) -> (Pose) {
    bLoop=false
    bResetPlayTimeWhenBlendSpaceChanges=false
    blendSpace=/GrappleWeapon/Items/Animations/GrappleKnife_Cast_Long_Loco_BS
    groupName=Cast
    method=SyncGroup
   }
   Output Animation Pose(Result=@n1)
  ]
 }
 Short {
  path=ItemFinalPoseOverride.AnimGraphNode_StateMachine_0.HasFired.AnimStateNode_0.CastFromVehicle.AnimGraphNode_StateMachine_1.CastUpperBody.AnimStateNode_3.Short
  nodes [
   n1 = Blendspace Player 'GrappleKnife_Cast_Short_UpperBody_BS'!(X=Get AimPitch()) -> (Pose) {bLoop=false, blendSpace=/GrappleWeapon/Items/Animations/GrappleKnife_Cast_Short_UpperBody_BS, groupName=Cast, method=SyncGroup}
   Output Animation Pose(Result=@n1)
  ]
 }
 Transition {
  path=ItemFinalPoseOverride.AnimGraphNode_StateMachine_0.HasFired.AnimStateNode_0.CastFromVehicle.AnimGraphNode_StateMachine_1.CastUpperBody.AnimStateTransitionNode_2.Transition
  nodes [Result(bCanEnterTransition=Get bDistanceBucketIs2() -> (bDistanceBucketIs2:bool))]
 }
 Transition {
  path=ItemFinalPoseOverride.AnimGraphNode_StateMachine_0.HasFired.AnimStateNode_0.CastFromVehicle.AnimGraphNode_StateMachine_1.CastUpperBody.AnimStateTransitionNode_3.Transition
  nodes [Result(bCanEnterTransition=true)]
 }
 Transition {
  path=ItemFinalPoseOverride.AnimGraphNode_StateMachine_0.HasFired.AnimStateNode_0.CastFromVehicle.AnimGraphNode_StateMachine_1.CastUpperBody.AnimStateTransitionNode_4.Transition
  nodes [Result(bCanEnterTransition=Get bDistanceBucketIs0() -> (bDistanceBucketIs0:bool))]
 }
 FullBodyState {
  path=ItemFinalPoseOverride.AnimGraphNode_StateMachine_0.HasFired.AnimStateNode_1.FullBodyState
  nodes [Output Animation Pose(Result=Use cached pose 'SavedFinalPose'())]
 }
 ZipAirAttack {
  path=ItemFinalPoseOverride.AnimGraphNode_StateMachine_0.HasFired.AnimStateNode_10.ZipAirAttack
  nodes [Output Animation Pose(Result=Modify Curve!(
     SourcePose=Sequence Player 'Grappleknife_Attack_InAir'() {bLoopAnimation=false, sequence=/GrappleWeapon/Items/Animations/Grappleknife_Attack_InAir}
     CurveValues_0=1
     CurveValues_1:float=1
     Alpha=1
    ) -> (Pose) {curveNames [Interruptible, BothHandFK], curveValues [0,0]})]
 }
 ZipUpperBody {
  path=ItemFinalPoseOverride.AnimGraphNode_StateMachine_0.HasFired.AnimStateNode_11.ZipUpperBody
  nodes [Output Animation Pose(Result=Modify Curve!(
     SourcePose=Sequence Player 'Grappleknife_Zip_Upperbody'() {bLoopAnimation=false, sequence=/GrappleWeapon/Items/Animations/Grappleknife_Zip_Upperbody}
     CurveValues_0=1
     Alpha=1
    ) -> (Pose) {curveNames [BothHandFK], curveValues [0]})]
 }
 ReleaseLoop {
  path=ItemFinalPoseOverride.AnimGraphNode_StateMachine_0.HasFired.AnimStateNode_12.ReleaseLoop
  nodes [
   n1 = Apply Additive(
    Base=Sequence Player 'Grappleknife_Release_Loop'() {bLoopAnimation=false, sequence=/GrappleWeapon/Items/Animations/Grappleknife_Release_Loop}
    Additive=Blendspace Player 'GrappleKnife_Releasel_Lean_BS'(X=Get PawnRotationRateYaw() -> (PawnRotationRateYaw:float), Y=0) {blendSpace=/GrappleWeapon/Items/Animations/GrappleKnife_Releasel_Lean_BS}
   ) {alphaBoolBlend {blendOption=Linear}}
   n2 = Get AimPitch()
   Output Animation Pose(Result=Modify Curve!(
     SourcePose=AimOffset Player 'GrappleKnife_Release_AO'!(X:float=0, Y:float=@n2, BasePose:Pose Link Structure=@n1, Alpha:float=1) -> (Pose:Pose Link Structure) {blendSpace=/GrappleWeapon/Items/Animations/GrappleKnife_Release_AO, alphaBoolBlend {blendOption=Linear}}
     CurveValues_0=1
     CurveValues_1:float=1
     Alpha=1
    ) -> (Pose) {curveNames [Interruptible, BothHandFK], curveValues [0,0]})
  ]
 }
 CastGroundLoco {
  path=ItemFinalPoseOverride.AnimGraphNode_StateMachine_0.HasFired.AnimStateNode_14.CastGroundLoco
  nodes [
   AnimGraphNode_StateMachine_1 = CastUpperBody() {bAllowConduitEntryStates=true}
   Output Animation Pose(Result=Modify Curve!(SourcePose=@AnimGraphNode_StateMachine_1, CurveValues_0=1, Alpha=1) -> (Pose) {curveNames [BothHandFK], curveValues [0]})
  ]
 }
 CastUpperBody {
  entry=Conduit
  path=ItemFinalPoseOverride.AnimGraphNode_StateMachine_0.HasFired.AnimStateNode_14.CastGroundLoco.AnimGraphNode_StateMachine_1.CastUpperBody
  ids=id
  nodes [
   AnimStateConduitNode_0 = Conduit()
   AnimStateNode_1 = Mid()
   AnimStateNode_2 = Long()
   AnimStateNode_3 = Short()
   AnimStateTransitionNode_2 = AnimStateTransitionNode() {crossfadeDuration=0.15, logicType=TLT_Inertialization, priorityOrder=2}
   AnimStateTransitionNode_3 = AnimStateTransitionNode() {crossfadeDuration=0.15, logicType=TLT_Inertialization, priorityOrder=100}
   AnimStateTransitionNode_4 = AnimStateTransitionNode() {logicType=TLT_Inertialization}
  ]
  links [
   {
    dst=Long
    rule=ItemFinalPoseOverride.AnimGraphNode_StateMachine_0.HasFired.AnimStateNode_14.CastGroundLoco.AnimGraphNode_StateMachine_1.CastUpperBody.AnimStateTransitionNode_2.Transition
    src=Conduit
    via=AnimStateTransitionNode_2
   }
   {
    dst=Mid
    rule=ItemFinalPoseOverride.AnimGraphNode_StateMachine_0.HasFired.AnimStateNode_14.CastGroundLoco.AnimGraphNode_StateMachine_1.CastUpperBody.AnimStateTransitionNode_3.Transition
    src=Conduit
    via=AnimStateTransitionNode_3
   }
   {
    dst=Short
    rule=ItemFinalPoseOverride.AnimGraphNode_StateMachine_0.HasFired.AnimStateNode_14.CastGroundLoco.AnimGraphNode_StateMachine_1.CastUpperBody.AnimStateTransitionNode_4.Transition
    src=Conduit
    via=AnimStateTransitionNode_4
   }
  ]
 }
 Conduit {
  path=ItemFinalPoseOverride.AnimGraphNode_StateMachine_0.HasFired.AnimStateNode_14.CastGroundLoco.AnimGraphNode_StateMachine_1.CastUpperBody.AnimStateConduitNode_0.Conduit
  nodes [Result(bCanEnterTransition=true)]
 }
 Mid {
  path=ItemFinalPoseOverride.AnimGraphNode_StateMachine_0.HasFired.AnimStateNode_14.CastGroundLoco.AnimGraphNode_StateMachine_1.CastUpperBody.AnimStateNode_1.Mid
  nodes [Output Animation Pose(Result=Blendspace Player 'GrappleKnife_Cast_Loco_BS'(X=Get AimPitch()) {bLoop=false, blendSpace=&GrappleKnife_Cast_Loco_BS, groupName=Cast, method=SyncGroup})]
 }
 Long {
  path=ItemFinalPoseOverride.AnimGraphNode_StateMachine_0.HasFired.AnimStateNode_14.CastGroundLoco.AnimGraphNode_StateMachine_1.CastUpperBody.AnimStateNode_2.Long
  nodes [
   n1 = Blendspace Player 'GrappleKnife_Cast_Long_Loco_BS'(X=Get AimPitch()) {
    bLoop=false
    bResetPlayTimeWhenBlendSpaceChanges=false
    blendSpace=&GrappleKnife_Cast_Long_Loco_BS
    groupName=Cast
    method=SyncGroup
   }
   Output Animation Pose(Result=@n1)
  ]
 }
 Short {
  path=ItemFinalPoseOverride.AnimGraphNode_StateMachine_0.HasFired.AnimStateNode_14.CastGroundLoco.AnimGraphNode_StateMachine_1.CastUpperBody.AnimStateNode_3.Short
  nodes [
   n1 = Blendspace Player 'GrappleKnife_Cast_Short_UpperBody_BS'(X=Get AimPitch()) {bLoop=false, blendSpace=&GrappleKnife_Cast_Short_UpperBody_BS, groupName=Cast, method=SyncGroup}
   n2 = Layered blend per bone(BasePose=@n1, BlendPoses_0=Use cached pose 'SavedFinalPose'()) {blendMode=BlendMask, blendMasks [&Fortnite_M_Avg_Player_Skeleton:LowerBody], layerSetup []}
   Output Animation Pose(Result=@n2)
  ]
 }
 Transition {
  path=ItemFinalPoseOverride.AnimGraphNode_StateMachine_0.HasFired.AnimStateNode_14.CastGroundLoco.AnimGraphNode_StateMachine_1.CastUpperBody.AnimStateTransitionNode_2.Transition
  nodes [Result(bCanEnterTransition=Get bDistanceBucketIs2())]
 }
 Transition {
  path=ItemFinalPoseOverride.AnimGraphNode_StateMachine_0.HasFired.AnimStateNode_14.CastGroundLoco.AnimGraphNode_StateMachine_1.CastUpperBody.AnimStateTransitionNode_3.Transition
  nodes [Result(bCanEnterTransition=true)]
 }
 Transition {
  path=ItemFinalPoseOverride.AnimGraphNode_StateMachine_0.HasFired.AnimStateNode_14.CastGroundLoco.AnimGraphNode_StateMachine_1.CastUpperBody.AnimStateTransitionNode_4.Transition
  nodes [Result(bCanEnterTransition=Get bDistanceBucketIs0())]
 }
 CastGroundIdle {
  path=ItemFinalPoseOverride.AnimGraphNode_StateMachine_0.HasFired.AnimStateNode_2.CastGroundIdle
  nodes [
   AnimGraphNode_StateMachine_0 = CastFullBody() {is=title AnimGraphNode_StateMachine, bAllowConduitEntryStates=true}
   Output Animation Pose(Result=Modify Curve(SourcePose=@AnimGraphNode_StateMachine_0, CurveValues_0=1, Alpha=1) {curveNames [BothHandFK], curveValues [0]})
  ]
 }
 CastFullBody {
  entry=Conduit
  path=ItemFinalPoseOverride.AnimGraphNode_StateMachine_0.HasFired.AnimStateNode_2.CastGroundIdle.AnimGraphNode_StateMachine_0.CastFullBody
  ids=id
  nodes [
   AnimStateConduitNode_0 = Conduit()
   AnimStateNode_0 = Short()
   AnimStateNode_1 = Mid()
   AnimStateNode_2 = Long()
   AnimStateTransitionNode_0 = AnimStateTransitionNode() {crossfadeDuration=0.15, logicType=TLT_Inertialization}
   AnimStateTransitionNode_1 = AnimStateTransitionNode() {crossfadeDuration=0.15, logicType=TLT_Inertialization, priorityOrder=100}
   AnimStateTransitionNode_2 = AnimStateTransitionNode() {crossfadeDuration=0.15, logicType=TLT_Inertialization, priorityOrder=2}
  ]
  links [
   {
    dst=Long
    rule=ItemFinalPoseOverride.AnimGraphNode_StateMachine_0.HasFired.AnimStateNode_2.CastGroundIdle.AnimGraphNode_StateMachine_0.CastFullBody.AnimStateTransitionNode_2.Transition
    src=Conduit
    via=AnimStateTransitionNode_2
   }
   {
    dst=Mid
    rule=ItemFinalPoseOverride.AnimGraphNode_StateMachine_0.HasFired.AnimStateNode_2.CastGroundIdle.AnimGraphNode_StateMachine_0.CastFullBody.AnimStateTransitionNode_1.Transition
    src=Conduit
    via=AnimStateTransitionNode_1
   }
   {
    dst=Short
    rule=ItemFinalPoseOverride.AnimGraphNode_StateMachine_0.HasFired.AnimStateNode_2.CastGroundIdle.AnimGraphNode_StateMachine_0.CastFullBody.AnimStateTransitionNode_0.Transition
    src=Conduit
    via=AnimStateTransitionNode_0
   }
  ]
 }
 Conduit {
  path=ItemFinalPoseOverride.AnimGraphNode_StateMachine_0.HasFired.AnimStateNode_2.CastGroundIdle.AnimGraphNode_StateMachine_0.CastFullBody.AnimStateConduitNode_0.Conduit
  nodes [Result(bCanEnterTransition=true)]
 }
 Short {
  path=ItemFinalPoseOverride.AnimGraphNode_StateMachine_0.HasFired.AnimStateNode_2.CastGroundIdle.AnimGraphNode_StateMachine_0.CastFullBody.AnimStateNode_0.Short
  nodes [
   n1 = Blendspace Player 'GrappleKnife_Cast_Short_FullBody_BS'!(X=Get AimPitch()) -> (Pose) {bLoop=false, blendSpace=/GrappleWeapon/Items/Animations/GrappleKnife_Cast_Short_FullBody_BS, groupName=Cast, method=SyncGroup}
   Output Animation Pose(Result=@n1)
  ]
 }
 Mid {
  path=ItemFinalPoseOverride.AnimGraphNode_StateMachine_0.HasFired.AnimStateNode_2.CastGroundIdle.AnimGraphNode_StateMachine_0.CastFullBody.AnimStateNode_1.Mid
  nodes [Output Animation Pose(Result=Blendspace Player 'GrappleKnife_Cast_FullBody_BS'!(X=Get AimPitch()) -> (Pose) {bLoop=false, blendSpace=/GrappleWeapon/Items/Animations/GrappleKnife_Cast_FullBody_BS, groupName=Cast, method=SyncGroup})]
 }
 Long {
  path=ItemFinalPoseOverride.AnimGraphNode_StateMachine_0.HasFired.AnimStateNode_2.CastGroundIdle.AnimGraphNode_StateMachine_0.CastFullBody.AnimStateNode_2.Long
  nodes [
   n1 = Blendspace Player 'GrappleKnife_Cast_Long_FullBody_BS'!(X=Get AimPitch()) -> (Pose) {
    bLoop=false
    bResetPlayTimeWhenBlendSpaceChanges=false
    blendSpace=/GrappleWeapon/Items/Animations/GrappleKnife_Cast_Long_FullBody_BS
    groupName=Cast
    method=SyncGroup
   }
   Output Animation Pose(Result=@n1)
  ]
 }
 Transition {
  path=ItemFinalPoseOverride.AnimGraphNode_StateMachine_0.HasFired.AnimStateNode_2.CastGroundIdle.AnimGraphNode_StateMachine_0.CastFullBody.AnimStateTransitionNode_0.Transition
  nodes [Result(bCanEnterTransition=Get bDistanceBucketIs0())]
 }
 Transition {
  path=ItemFinalPoseOverride.AnimGraphNode_StateMachine_0.HasFired.AnimStateNode_2.CastGroundIdle.AnimGraphNode_StateMachine_0.CastFullBody.AnimStateTransitionNode_1.Transition
  nodes [Result(bCanEnterTransition=true)]
 }
 Transition {
  path=ItemFinalPoseOverride.AnimGraphNode_StateMachine_0.HasFired.AnimStateNode_2.CastGroundIdle.AnimGraphNode_StateMachine_0.CastFullBody.AnimStateTransitionNode_2.Transition
  nodes [Result(bCanEnterTransition=Get bDistanceBucketIs2())]
 }
 CastUpperBody {
  path=ItemFinalPoseOverride.AnimGraphNode_StateMachine_0.HasFired.AnimStateNode_3.CastUpperBody
  nodes [
   AnimGraphNode_StateMachine_0 = CastFullBody() {bAllowConduitEntryStates=true}
   n1 = Layered blend per bone(BasePose=@AnimGraphNode_StateMachine_0, BlendPoses_0=Use cached pose 'SavedFinalPose'()) {blendMode=BlendMask, blendMasks [&Fortnite_M_Avg_Player_Skeleton:LowerBody], layerSetup []}
   Output Animation Pose(Result=Modify Curve(SourcePose=@n1, CurveValues_0=1, Alpha=1) {curveNames [BothHandFK], curveValues [0]})
  ]
 }
 CastFullBody {
  entry=Conduit
  path=ItemFinalPoseOverride.AnimGraphNode_StateMachine_0.HasFired.AnimStateNode_3.CastUpperBody.AnimGraphNode_StateMachine_0.CastFullBody
  ids=id
  nodes [
   AnimStateConduitNode_0 = Conduit()
   AnimStateNode_0 = Short()
   AnimStateNode_1 = Mid()
   AnimStateNode_2 = Long()
   AnimStateTransitionNode_0 = AnimStateTransitionNode() {crossfadeDuration=0.15}
   AnimStateTransitionNode_1 = AnimStateTransitionNode() {crossfadeDuration=0.15, priorityOrder=100}
   AnimStateTransitionNode_2 = AnimStateTransitionNode() {crossfadeDuration=0.15, priorityOrder=2}
  ]
  links [
   {
    dst=Long
    rule=ItemFinalPoseOverride.AnimGraphNode_StateMachine_0.HasFired.AnimStateNode_3.CastUpperBody.AnimGraphNode_StateMachine_0.CastFullBody.AnimStateTransitionNode_2.Transition
    src=Conduit
    via=AnimStateTransitionNode_2
   }
   {
    dst=Mid
    rule=ItemFinalPoseOverride.AnimGraphNode_StateMachine_0.HasFired.AnimStateNode_3.CastUpperBody.AnimGraphNode_StateMachine_0.CastFullBody.AnimStateTransitionNode_1.Transition
    src=Conduit
    via=AnimStateTransitionNode_1
   }
   {
    dst=Short
    rule=ItemFinalPoseOverride.AnimGraphNode_StateMachine_0.HasFired.AnimStateNode_3.CastUpperBody.AnimGraphNode_StateMachine_0.CastFullBody.AnimStateTransitionNode_0.Transition
    src=Conduit
    via=AnimStateTransitionNode_0
   }
  ]
 }
 Conduit {
  path=ItemFinalPoseOverride.AnimGraphNode_StateMachine_0.HasFired.AnimStateNode_3.CastUpperBody.AnimGraphNode_StateMachine_0.CastFullBody.AnimStateConduitNode_0.Conduit
  nodes [Result(bCanEnterTransition=true)]
 }
 Short {
  path=ItemFinalPoseOverride.AnimGraphNode_StateMachine_0.HasFired.AnimStateNode_3.CastUpperBody.AnimGraphNode_StateMachine_0.CastFullBody.AnimStateNode_0.Short
  nodes [
   n1 = Blendspace Player 'GrappleKnife_Cast_Short_UpperBody_BS'(X=Get AimPitch()) {bLoop=false, blendSpace=&GrappleKnife_Cast_Short_UpperBody_BS, groupName=Cast, method=SyncGroup}
   Output Animation Pose(Result=@n1)
  ]
 }
 Mid {
  path=ItemFinalPoseOverride.AnimGraphNode_StateMachine_0.HasFired.AnimStateNode_3.CastUpperBody.AnimGraphNode_StateMachine_0.CastFullBody.AnimStateNode_1.Mid
  nodes [
   n1 = Blendspace Player 'GrappleKnife_Cast_Mid_UpperBody_BS'!(X=Get AimPitch()) -> (Pose) {bLoop=false, blendSpace=/GrappleWeapon/Items/Animations/GrappleKnife_Cast_Mid_UpperBody_BS, groupName=Cast, method=SyncGroup}
   Output Animation Pose(Result=@n1)
  ]
 }
 Long {
  path=ItemFinalPoseOverride.AnimGraphNode_StateMachine_0.HasFired.AnimStateNode_3.CastUpperBody.AnimGraphNode_StateMachine_0.CastFullBody.AnimStateNode_2.Long
  nodes [
   n1 = Blendspace Player 'GrappleKnife_Cast_Long_UpperBody_BS1'!(X=Get AimPitch()) -> (Pose) {bLoop=false, blendSpace=/GrappleWeapon/Items/Animations/GrappleKnife_Cast_Long_UpperBody_BS1, groupName=Cast, method=SyncGroup}
   Output Animation Pose(Result=@n1)
  ]
 }
 Transition {
  path=ItemFinalPoseOverride.AnimGraphNode_StateMachine_0.HasFired.AnimStateNode_3.CastUpperBody.AnimGraphNode_StateMachine_0.CastFullBody.AnimStateTransitionNode_0.Transition
  nodes [Result(bCanEnterTransition=Get bDistanceBucketIs0())]
 }
 Transition {
  path=ItemFinalPoseOverride.AnimGraphNode_StateMachine_0.HasFired.AnimStateNode_3.CastUpperBody.AnimGraphNode_StateMachine_0.CastFullBody.AnimStateTransitionNode_1.Transition
  nodes [Result(bCanEnterTransition=true)]
 }
 Transition {
  path=ItemFinalPoseOverride.AnimGraphNode_StateMachine_0.HasFired.AnimStateNode_3.CastUpperBody.AnimGraphNode_StateMachine_0.CastFullBody.AnimStateTransitionNode_2.Transition
  nodes [Result(bCanEnterTransition=Get bDistanceBucketIs2())]
 }
 Fall {
  path=ItemFinalPoseOverride.AnimGraphNode_StateMachine_0.HasFired.AnimStateNode_4.Fall
  nodes [
   n1 = Apply Additive(
    Base=Sequence Player 'Grappleknife_Fall_Idle'() {sequence=/GrappleWeapon/Items/Animations/Grappleknife_Fall_Idle}
    Additive=Blendspace Player 'GrappleKnife_Fall_Lean_BS'(X=Get PawnRotationRateYaw(), Y=0) {blendSpace=/GrappleWeapon/Items/Animations/GrappleKnife_Fall_Lean_BS}
   ) {alphaBoolBlend {blendOption=Linear}}
   n2 = Get AimPitch()
   Output Animation Pose(Result=Modify Curve!(
     SourcePose=AimOffset Player 'GrappleKnife_Release_Fall_AO'!(X:float=0, Y:float=@n2, BasePose:Pose Link Structure=@n1, Alpha:float=1) -> (Pose:Pose Link Structure) {blendSpace=/GrappleWeapon/Items/Animations/GrappleKnife_Release_Fall_AO, alphaBoolBlend {blendOption=Linear}}
     CurveValues_0=1
     CurveValues_1:float=1
     Alpha=1
    ) -> (Pose) {curveNames [Interruptible, BothHandFK], curveValues [0,0]})
  ]
 }
 Release {
  path=ItemFinalPoseOverride.AnimGraphNode_StateMachine_0.HasFired.AnimStateNode_5.Release
  nodes [Output Animation Pose(Result=Modify Curve(
     SourcePose=Sequence Player 'Grappleknife_Release_01'() {bLoopAnimation=false, sequence=/GrappleWeapon/Items/Animations/Grappleknife_Release_01}
     CurveValues_0=1
     CurveValues_1=1
     Alpha=1
    ) {curveNames [Interruptible, BothHandFK], curveValues [0,0]})]
 }
 CastInAir {
  path=ItemFinalPoseOverride.AnimGraphNode_StateMachine_0.HasFired.AnimStateNode_6.CastInAir
  nodes [
   n1 = Get AimPitch()
   Output Animation Pose(Result=Modify Curve!(
     SourcePose=Blendspace Player 'GrappleKnife_Cast_InAir_BS'!(X=@n1) -> (Pose) {bLoop=false, blendSpace=/GrappleWeapon/Items/Animations/GrappleKnife_Cast_InAir_BS}
     CurveValues_0=1
     Alpha=1
    ) -> (Pose) {curveNames [BothHandFK], curveValues [0]})
  ]
 }
 ZipAirStart {
  path=ItemFinalPoseOverride.AnimGraphNode_StateMachine_0.HasFired.AnimStateNode_7.ZipAirStart
  nodes [
   n1 = Get AimPitch()
   Output Animation Pose(Result=Modify Curve(
     SourcePose=Blendspace Player 'GrappleKnife_Zip_Start_InAir_BS'!(X=@n1) -> (Pose) {blendSpace=/GrappleWeapon/Items/Animations/GrappleKnife_Zip_Start_InAir_BS}
     CurveValues_0=1
     Alpha=1
    ) {curveNames [BothHandFK], curveValues [0]})
  ]
 }
 ZipGroundStart {
  path=ItemFinalPoseOverride.AnimGraphNode_StateMachine_0.HasFired.AnimStateNode_8.ZipGroundStart
  nodes [Output Animation Pose(Result=Modify Curve(
     SourcePose=Sequence Player 'GrappleKnife_Zip'() {bLoopAnimation=false, groupName=ZipStart, method=SyncGroup, sequence=/GrappleWeapon/Items/Animations/GrappleKnife_Zip}
     CurveValues_0=1
     Alpha=1
    ) {curveNames [BothHandFK], curveValues [0]})]
 }
 WallBounce {
  path=ItemFinalPoseOverride.AnimGraphNode_StateMachine_0.HasFired.AnimStateNode_9.WallBounce
  nodes [
   n1 = Blendspace Player 'GrappleKnife_Bounce_BS'!(X=Get AnimBounceAngle() -> (AnimBounceAngle:float)) -> (Pose) {bLoop=false, blendSpace=/GrappleWeapon/Items/Animations/GrappleKnife_Bounce_BS}
   n2 = Get AimPitch()
   Output Animation Pose(Result=Modify Curve!(
     SourcePose=AimOffset Player 'GrappleKnife_Release_AO'(X=0, Y=@n2, BasePose=@n1, Alpha=1) {blendSpace=&GrappleKnife_Release_AO, alphaBoolBlend {blendOption=Linear}}
     CurveValues_0=1
     CurveValues_1:float=1
     Alpha=1
    ) -> (Pose) {curveNames [Interruptible, BothHandFK], curveValues [0,0]})
  ]
 }
 Transition {
  path=ItemFinalPoseOverride.AnimGraphNode_StateMachine_0.HasFired.AnimStateTransitionNode_0.Transition
  nodes [Result(bCanEnterTransition=Get bIsCastGroundIdle() -> (bIsCastGroundIdle:bool))]
 }
 Transition {
  path=ItemFinalPoseOverride.AnimGraphNode_StateMachine_0.HasFired.AnimStateTransitionNode_1.Transition
  nodes [Result(bCanEnterTransition=Get bIsFireOnMotorBike() -> (bIsFireOnMotorBike:bool))]
 }
 Transition {
  path=ItemFinalPoseOverride.AnimGraphNode_StateMachine_0.HasFired.AnimStateTransitionNode_10.Transition
  nodes [
   Result()
   AND Boolean(
    A=NOT Boolean(A=Get!() -> (bIsAttached:bool, Is Attached_b:bool).Is Attached_b)
    B="float > float"(A=GetCurveValue(CurveName=StopAnim), B=0.5)
   )
  ]
 }
 Transition {
  path=ItemFinalPoseOverride.AnimGraphNode_StateMachine_0.HasFired.AnimStateTransitionNode_11.Transition
  nodes [Result(bCanEnterTransition=Get bIsAccelerating2D())]
 }
 Transition {path=ItemFinalPoseOverride.AnimGraphNode_StateMachine_0.HasFired.AnimStateTransitionNode_12.Transition, nodes [Result()]}
 Transition {
  path=ItemFinalPoseOverride.AnimGraphNode_StateMachine_0.HasFired.AnimStateTransitionNode_13.Transition
  nodes [Result(bCanEnterTransition=Get bIsCastUpperbody() -> (bIsCastUpperbody:bool))]
 }
 Transition {
  path=ItemFinalPoseOverride.AnimGraphNode_StateMachine_0.HasFired.AnimStateTransitionNode_14.Transition
  nodes [
   Result(bCanEnterTransition=Get bIsExitZipAirStart() -> (bIsExitZipAirStart:bool))
   AND Boolean(A=NOT Boolean(A=Get().Is Attached_b), B="float > float"(A=GetCurveValue(CurveName=StopAnim), B=0.5))
  ]
 }
 Transition {
  path=ItemFinalPoseOverride.AnimGraphNode_StateMachine_0.HasFired.AnimStateTransitionNode_15.Transition
  nodes [Result(bCanEnterTransition=true)]
 }
 Transition {
  path=ItemFinalPoseOverride.AnimGraphNode_StateMachine_0.HasFired.AnimStateTransitionNode_16.Transition
  nodes [Result(bCanEnterTransition=Get bAnimStartWallBounce() -> (bAnimStartWallBounce:bool))]
 }
 Transition {
  path=ItemFinalPoseOverride.AnimGraphNode_StateMachine_0.HasFired.AnimStateTransitionNode_17.Transition
  nodes [Result(bCanEnterTransition=Get bIsExitUnexpectedly() -> (bIsExitUnexpectedly:bool))]
 }
 Transition {path=ItemFinalPoseOverride.AnimGraphNode_StateMachine_0.HasFired.AnimStateTransitionNode_18.Transition, nodes [Result()]}
 Transition {path=ItemFinalPoseOverride.AnimGraphNode_StateMachine_0.HasFired.AnimStateTransitionNode_19.Transition, nodes [Result()]}
 Transition {
  path=ItemFinalPoseOverride.AnimGraphNode_StateMachine_0.HasFired.AnimStateTransitionNode_2.Transition
  nodes [Result(bCanEnterTransition=Get bIsCastGroundLoco() -> (bIsCastGroundLoco:bool))]
 }
 Transition {path=ItemFinalPoseOverride.AnimGraphNode_StateMachine_0.HasFired.AnimStateTransitionNode_20.Transition, nodes [Result()]}
 Transition {path=ItemFinalPoseOverride.AnimGraphNode_StateMachine_0.HasFired.AnimStateTransitionNode_21.Transition, nodes [Result()]}
 Transition {path=ItemFinalPoseOverride.AnimGraphNode_StateMachine_0.HasFired.AnimStateTransitionNode_22.Transition, nodes [Result()]}
 Transition {
  path=ItemFinalPoseOverride.AnimGraphNode_StateMachine_0.HasFired.AnimStateTransitionNode_23.Transition
  nodes [
   Result(bCanEnterTransition=Get bIsExitWallBounce() -> (bIsExitWallBounce:bool))
   "float >= float"(A=GetCurveValue(CurveName=BlendOut), B=0.5)
  ]
 }
 Transition {
  path=ItemFinalPoseOverride.AnimGraphNode_StateMachine_0.HasFired.AnimStateTransitionNode_3.Transition
  nodes [Result(bCanEnterTransition=NOT Boolean(A=Get bIsAccelerating2D()))]
 }
 Transition {path=ItemFinalPoseOverride.AnimGraphNode_StateMachine_0.HasFired.AnimStateTransitionNode_4.Transition, nodes [Result()]}
 Transition {
  path=ItemFinalPoseOverride.AnimGraphNode_StateMachine_0.HasFired.AnimStateTransitionNode_5.Transition
  nodes [Result(bCanEnterTransition=Get bIsFired())]
 }
 Transition {
  path=ItemFinalPoseOverride.AnimGraphNode_StateMachine_0.HasFired.AnimStateTransitionNode_6.Transition
  nodes [Result(bCanEnterTransition=Get bIsCastInAir() -> (bIsCastInAir:bool))]
 }
 Transition {
  path=ItemFinalPoseOverride.AnimGraphNode_StateMachine_0.HasFired.AnimStateTransitionNode_7.Transition
  nodes [
   Result(bCanEnterTransition=Get bIsExitGround() -> (bIsExitGround:bool))
   AND Boolean!(
    A=NOT Boolean(A=Get bIsJumping())
    B=NOT Boolean(A=Get bIsFalling())
    C:bool="float > float"(A=GetCurveValue(CurveName=Interruptible), B=0.5)
   ) -> (ReturnValue)
  ]
 }
 Transition {
  path=ItemFinalPoseOverride.AnimGraphNode_StateMachine_0.HasFired.AnimStateTransitionNode_8.Transition
  nodes [Result(bCanEnterTransition=true)]
 }
 Transition {
  path=ItemFinalPoseOverride.AnimGraphNode_StateMachine_0.HasFired.AnimStateTransitionNode_9.Transition
  nodes [Result(bCanEnterTransition=Get IsInAirAttack() -> (IsInAirAttack:bool))]
 }
 ItemFullBodyAimOffsets {nodes [
   Input Pose!() -> (Pose, FullBody_AimOffsetAlpha:double, Aim_Yaw:double, Aim_Pitch:double) {name=InPoseFullBodyAimOffsets}
   Input Pose!() -> (Pose) {name=InPosePreFullBodyAimOffsets}
   Output Pose() {layerGroup=Item Override}
  ]}
 ItemFullBodyOverride {nodes [
   Input Pose() {name=InPoseFullBody}
   Output Pose() {layerGroup=Item Override}
  ]}
 ItemIdleOverride {nodes [
   Input Pose() {name=InPose_Idle}
   Output Pose() {layerGroup=Item Override}
  ]}
 ItemJetPackJumpAdditive {nodes [Output Pose() {layerGroup=Item Override}]}
 ItemJetPackStartAdditive {nodes [Output Pose() {layerGroup=Item Override}]}
 ItemJumpLoopAdditive {nodes [
   Input Pose() {name=InPoseItemJumpLoopAdditive}
   Output Pose() {layerGroup=Item Override}
  ]}
 ItemJumpOverride {nodes [
   Input Pose() {name=InPoseJump}
   Output Pose() {layerGroup=Item Override}
  ]}
 ItemJumpUpAdditive {nodes [
   Input Pose() {name=InPoseItemJumpUpAdditive}
   Output Pose() {layerGroup=Item Override}
  ]}
 ItemPreAimOffsets {nodes [
   Input Pose() {name=InPosePreAimOffsets}
   Output Pose() {layerGroup=Item Override}
  ]}
 ItemPreIK {nodes [
   Input Pose() {name=InPose_PreIK}
   Output Pose() {layerGroup=Item Override}
  ]}
 ItemRightHandIK {nodes [
   Input Pose() {name=InPosePreRightHandIK}
   Input Pose() {name=InPosePostRightHandIK}
   Output Pose() {layerGroup=Item Override}
  ]}
 ItemSkeletalControl {nodes [
   Input Pose()
   Output Pose() {layerGroup=Item Override}
  ]}
 ItemSwimJumpFallLoop {nodes [Output Pose() {layerGroup=Item Override}]}
 ItemSwimJumpSurfaceEnd {nodes [Output Pose() {layerGroup=Item Override}]}
 ItemSwimJumpSurfaceLoop {nodes [Output Pose() {layerGroup=Item Override}]}
 ItemUpperBody {nodes [
   Input Pose!() -> (Pose, InputParam:Fort Anim Input Adjusted Aim Structure) {name=InPoseUpperBody}
   Output Pose() {layerGroup=Item Override}
  ]}
 ItemZipLineStartAdditive {nodes [Output Pose() {layerGroup=Item Override}]}
]
variables [IsInAirAttack {type=boolean}]
