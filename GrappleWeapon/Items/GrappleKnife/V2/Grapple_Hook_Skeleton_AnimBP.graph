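# Animation Blueprint for the grapple knife's hook mesh, playing its idle, melee attack and cast animations.
# A conduit hub routes to air, full-body and upper-body attacks and three cast ranges (mid reuses the long clip), mostly via inertialized transitions, before DefaultSlot and Dead Blending.
# Idle shows an open or closed pose from the armed flag.
# Each update copies armed, fired and distance-bucket values from the pawn's current B_GrappleKnife.
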AnimBlueprint /GrappleWeapon/Items/GrappleKnife/V2/Grapple_Hook_Skeleton_AnimBP
target_skeleton=/GrappleWeapon/Meshes/Grapple_KnifeHook_Sharp_Skeleton
graphs [
 AnimGraph {nodes [
   AnimGraphNode_StateMachine_0 = Grapple Anim() {is=title AnimGraphNode_StateMachine, bAllowConduitEntryStates=true}
   Output Pose(Result=Dead Blending!(Source:Pose Link Structure=Slot 'DefaultSlot'(Source=@AnimGraphNode_StateMachine_0)) -> (Pose:Pose Link Structure) {is=title AnimGraphNode_DeadBlending, defaultBlendMode=Linear}) {layerGroup=None}
  ]}
 Grapple Anim {
  entry=Conduit
  path=AnimGraph.AnimGraphNode_StateMachine_0.Grapple Anim
  ids=id
  nodes [
   FromRelease() {is=title AnimStateAliasNode}
   AnimStateConduitNode_0 = Conduit()
   AnimStateNode_0 = Idle()
   AnimStateNode_1 = AirAttack() {is=title AnimStateNode}
   AnimStateNode_2 = FullbodyAttack_1() {is=title AnimStateNode}
   AnimStateNode_3 = FullbodyAttack_2() {is=title AnimStateNode}
   AnimStateNode_4 = CastShort() {
    is=title AnimStateNode
    stateEntered {guid=#1}
    stateFullyBlended {guid=#2}
    stateLeft {guid=#3}
   }
   AnimStateNode_6 = UpperBodyAttack_1() {is=title AnimStateNode}
   AnimStateNode_7 = UpperBodyAttack_2() {is=title AnimStateNode}
   AnimStateNode_8 = CastLong() {
    is=title AnimStateNode
    stateEntered {guid=#1}
    stateFullyBlended {guid=#2}
    stateLeft {guid=#3}
   }
   AnimStateNode_9 = CastLMid() {
    is=title AnimStateNode
    stateEntered {guid=#1}
    stateFullyBlended {guid=#2}
    stateLeft {guid=#3}
   }
   AnimStateTransitionNode_0 = AnimStateTransitionNode() {automaticRuleTriggerTime=0, bAutomaticRuleBasedOnSequencePlayerInState=true}
   AnimStateTransitionNode_1 = AnimStateTransitionNode() {crossfadeDuration=0.1, logicType=TLT_Inertialization, priorityOrder=4}
   AnimStateTransitionNode_10 = AnimStateTransitionNode() {automaticRuleTriggerTime=0, bAutomaticRuleBasedOnSequencePlayerInState=true}
   AnimStateTransitionNode_11 = AnimStateTransitionNode() {automaticRuleTriggerTime=0, bAutomaticRuleBasedOnSequencePlayerInState=true}
   AnimStateTransitionNode_12 = AnimStateTransitionNode() {automaticRuleTriggerTime=0, bAutomaticRuleBasedOnSequencePlayerInState=true}
   AnimStateTransitionNode_13 = AnimStateTransitionNode() {crossfadeDuration=0.1, logicType=TLT_Inertialization}
   AnimStateTransitionNode_14 = AnimStateTransitionNode() {crossfadeDuration=0.1, logicType=TLT_Inertialization}
   AnimStateTransitionNode_15 = AnimStateTransitionNode()
   AnimStateTransitionNode_16 = AnimStateTransitionNode() {crossfadeDuration=0.3, logicType=TLT_Inertialization, priorityOrder=100}
   AnimStateTransitionNode_17 = AnimStateTransitionNode() {automaticRuleTriggerTime=0, bAutomaticRuleBasedOnSequencePlayerInState=true}
   AnimStateTransitionNode_18 = AnimStateTransitionNode() {automaticRuleTriggerTime=0, bAutomaticRuleBasedOnSequencePlayerInState=true}
   AnimStateTransitionNode_2 = AnimStateTransitionNode() {crossfadeDuration=0.1, logicType=TLT_Inertialization}
   AnimStateTransitionNode_27 = AnimStateTransitionNode() {automaticRuleTriggerTime=0, bAutomaticRuleBasedOnSequencePlayerInState=true}
   AnimStateTransitionNode_28 = AnimStateTransitionNode()
   AnimStateTransitionNode_3 = AnimStateTransitionNode() {crossfadeDuration=0.1, logicType=TLT_Inertialization}
   AnimStateTransitionNode_32 = AnimStateTransitionNode() {logicType=TLT_Inertialization, priorityOrder=8}
   AnimStateTransitionNode_33 = AnimStateTransitionNode() {logicType=TLT_Inertialization, priorityOrder=7}
   AnimStateTransitionNode_34 = AnimStateTransitionNode() {logicType=TLT_Inertialization, priorityOrder=6}
   AnimStateTransitionNode_35 = AnimStateTransitionNode()
   AnimStateTransitionNode_36 = AnimStateTransitionNode()
   AnimStateTransitionNode_37 = AnimStateTransitionNode()
   AnimStateTransitionNode_4 = AnimStateTransitionNode() {crossfadeDuration=0.1, logicType=TLT_Inertialization}
   AnimStateTransitionNode_5 = AnimStateTransitionNode() {automaticRuleTriggerTime=0, bAutomaticRuleBasedOnSequencePlayerInState=true}
   AnimStateTransitionNode_6 = AnimStateTransitionNode() {crossfadeDuration=0.1, logicType=TLT_Inertialization, priorityOrder=2}
   AnimStateTransitionNode_7 = AnimStateTransitionNode() {crossfadeDuration=0.1, logicType=TLT_Inertialization, priorityOrder=5}
   AnimStateTransitionNode_8 = AnimStateTransitionNode() {crossfadeDuration=0.1, logicType=TLT_Inertialization, priorityOrder=3}
   AnimStateTransitionNode_9 = AnimStateTransitionNode() {automaticRuleTriggerTime=0, bAutomaticRuleBasedOnSequencePlayerInState=true}
  ]
  links [
   {
    dst=Conduit
    rule=AnimGraph.AnimGraphNode_StateMachine_0.Grapple Anim.AnimStateTransitionNode_5.Transition
    src=AirAttack
    via=AnimStateTransitionNode_5
   }
   {
    dst=Conduit
    rule=AnimGraph.AnimGraphNode_StateMachine_0.Grapple Anim.AnimStateTransitionNode_11.Transition
    src=CastLMid
    via=AnimStateTransitionNode_11
   }
   {
    dst=Conduit
    rule=AnimGraph.AnimGraphNode_StateMachine_0.Grapple Anim.AnimStateTransitionNode_36.Transition
    src=CastLMid
    via=AnimStateTransitionNode_36
   }
   {
    dst=Conduit
    rule=AnimGraph.AnimGraphNode_StateMachine_0.Grapple Anim.AnimStateTransitionNode_0.Transition
    src=CastLong
    via=AnimStateTransitionNode_0
   }
   {
    dst=Conduit
    rule=AnimGraph.AnimGraphNode_StateMachine_0.Grapple Anim.AnimStateTransitionNode_37.Transition
    src=CastLong
    via=AnimStateTransitionNode_37
   }
   {
    dst=Conduit
    rule=AnimGraph.AnimGraphNode_StateMachine_0.Grapple Anim.AnimStateTransitionNode_12.Transition
    src=CastShort
    via=AnimStateTransitionNode_12
   }
   {
    dst=Conduit
    rule=AnimGraph.AnimGraphNode_StateMachine_0.Grapple Anim.AnimStateTransitionNode_35.Transition
    src=CastShort
    via=AnimStateTransitionNode_35
   }
   {
    dst=AirAttack
    rule=AnimGraph.AnimGraphNode_StateMachine_0.Grapple Anim.AnimStateTransitionNode_4.Transition
    src=Conduit
    via=AnimStateTransitionNode_4
   }
   {
    dst=CastLMid
    rule=AnimGraph.AnimGraphNode_StateMachine_0.Grapple Anim.AnimStateTransitionNode_33.Transition
    src=Conduit
    via=AnimStateTransitionNode_33
   }
   {
    dst=CastLong
    rule=AnimGraph.AnimGraphNode_StateMachine_0.Grapple Anim.AnimStateTransitionNode_34.Transition
    src=Conduit
    via=AnimStateTransitionNode_34
   }
   {
    dst=CastShort
    rule=AnimGraph.AnimGraphNode_StateMachine_0.Grapple Anim.AnimStateTransitionNode_32.Transition
    src=Conduit
    via=AnimStateTransitionNode_32
   }
   {
    dst=FullbodyAttack_1
    rule=AnimGraph.AnimGraphNode_StateMachine_0.Grapple Anim.AnimStateTransitionNode_6.Transition
    src=Conduit
    via=AnimStateTransitionNode_6
   }
   {
    dst=FullbodyAttack_2
    rule=AnimGraph.AnimGraphNode_StateMachine_0.Grapple Anim.AnimStateTransitionNode_8.Transition
    src=Conduit
    via=AnimStateTransitionNode_8
   }
   {
    dst=Idle
    rule=AnimGraph.AnimGraphNode_StateMachine_0.Grapple Anim.AnimStateTransitionNode_16.Transition
    src=Conduit
    via=AnimStateTransitionNode_16
   }
   {
    dst=UpperBodyAttack_1
    rule=AnimGraph.AnimGraphNode_StateMachine_0.Grapple Anim.AnimStateTransitionNode_1.Transition
    src=Conduit
    via=AnimStateTransitionNode_1
   }
   {
    dst=UpperBodyAttack_2
    rule=AnimGraph.AnimGraphNode_StateMachine_0.Grapple Anim.AnimStateTransitionNode_7.Transition
    src=Conduit
    via=AnimStateTransitionNode_7
   }
   {
    dst=Conduit
    rule=AnimGraph.AnimGraphNode_StateMachine_0.Grapple Anim.AnimStateTransitionNode_27.Transition
    src=FromRelease
    via=AnimStateTransitionNode_27
   }
   {
    dst=Conduit
    rule=AnimGraph.AnimGraphNode_StateMachine_0.Grapple Anim.AnimStateTransitionNode_28.Transition
    src=FromRelease
    via=AnimStateTransitionNode_28
   }
   {
    dst=Conduit
    rule=AnimGraph.AnimGraphNode_StateMachine_0.Grapple Anim.AnimStateTransitionNode_9.Transition
    src=FullbodyAttack_1
    via=AnimStateTransitionNode_9
   }
   {
    dst=FullbodyAttack_2
    rule=AnimGraph.AnimGraphNode_StateMachine_0.Grapple Anim.AnimStateTransitionNode_2.Transition
    src=FullbodyAttack_1
    via=AnimStateTransitionNode_2
   }
   {
    dst=Conduit
    rule=AnimGraph.AnimGraphNode_StateMachine_0.Grapple Anim.AnimStateTransitionNode_10.Transition
    src=FullbodyAttack_2
    via=AnimStateTransitionNode_10
   }
   {
    dst=FullbodyAttack_1
    rule=AnimGraph.AnimGraphNode_StateMachine_0.Grapple Anim.AnimStateTransitionNode_3.Transition
    src=FullbodyAttack_2
    via=AnimStateTransitionNode_3
   }
   {
    dst=Conduit
    rule=AnimGraph.AnimGraphNode_StateMachine_0.Grapple Anim.AnimStateTransitionNode_15.Transition
    src=Idle
    via=AnimStateTransitionNode_15
   }
   {
    dst=Conduit
    rule=AnimGraph.AnimGraphNode_StateMachine_0.Grapple Anim.AnimStateTransitionNode_18.Transition
    src=UpperBodyAttack_1
    via=AnimStateTransitionNode_18
   }
   {
    dst=UpperBodyAttack_2
    rule=AnimGraph.AnimGraphNode_StateMachine_0.Grapple Anim.AnimStateTransitionNode_13.Transition
    src=UpperBodyAttack_1
    via=AnimStateTransitionNode_13
   }
   {
    dst=Conduit
    rule=AnimGraph.AnimGraphNode_StateMachine_0.Grapple Anim.AnimStateTransitionNode_17.Transition
    src=UpperBodyAttack_2
    via=AnimStateTransitionNode_17
   }
   {
    dst=UpperBodyAttack_1
    rule=AnimGraph.AnimGraphNode_StateMachine_0.Grapple Anim.AnimStateTransitionNode_14.Transition
    src=UpperBodyAttack_2
    via=AnimStateTransitionNode_14
   }
  ]
 }
 Conduit {path=AnimGraph.AnimGraphNode_StateMachine_0.Grapple Anim.AnimStateConduitNode_0.Conduit, nodes [Result(bCanEnterTransition=true)]}
 Idle {
  path=AnimGraph.AnimGraphNode_StateMachine_0.Grapple Anim.AnimStateNode_0.Idle
  nodes [Output Animation Pose(Result=Blend Poses by bool(
     BlendPose_0="Sequence Evaluator (by time) 'GrappleKnife_Hook_toOpen_WEP_hook'"!(ExplicitTime:float=0.67) -> (Pose:Pose Link Structure) {is=title AnimGraphNode_SequenceEvaluator, sequence=/GrappleWeapon/Items/Animations/WEP/GrappleKnife_Hook_toOpen_WEP_hook}
     BlendPose_1=Sequence Player 'GrappleKnife_Hook_Closed_Idle_WEP_hook'() {bLoopAnimation=false, sequence=/GrappleWeapon/Items/Animations/WEP/GrappleKnife_Hook_Closed_Idle_WEP_hook}
     BlendTime_0=0.2
     BlendTime_1=0.3
     bActiveValue=Get Is Grapple Armed() -> (Is Grapple Armed:bool)
    ) {blendType=Linear, transitionType=Inertialization})]
 }
 AirAttack {
  path=AnimGraph.AnimGraphNode_StateMachine_0.Grapple Anim.AnimStateNode_1.AirAttack
  nodes [Output Animation Pose(Result=Sequence Player 'GrappleKnife_inAirAttack_WEP_hook'() {bLoopAnimation=false, sequence=/GrappleWeapon/Items/Animations/WEP/GrappleKnife_inAirAttack_WEP_hook})]
 }
 FullbodyAttack_1 {
  path=AnimGraph.AnimGraphNode_StateMachine_0.Grapple Anim.AnimStateNode_2.FullbodyAttack_1
  nodes [
   Sequence Player 'GrappleKnife_GroundAttack_01_male_WEP_hook'() {bLoopAnimation=false, sequence=/GrappleWeapon/Items/Animations/WEP/GrappleKnife_GroundAttack_01_male_WEP_hook}
   Output Animation Pose(Result=Sequence Player 'GrappleKnife_GroundAttack_01_male_WEP_hook'() {bLoopAnimation=false, sequence=&GrappleKnife_GroundAttack_01_male_WEP_hook})
  ]
 }
 FullbodyAttack_2 {
  path=AnimGraph.AnimGraphNode_StateMachine_0.Grapple Anim.AnimStateNode_3.FullbodyAttack_2
  nodes [
   Sequence Player 'GrappleKnife_GroundAttack_02_male_WEP_hook'() {bLoopAnimation=false, sequence=/GrappleWeapon/Items/Animations/WEP/GrappleKnife_GroundAttack_02_male_WEP_hook}
   Output Animation Pose(Result=Sequence Player 'GrappleKnife_GroundAttack_02_male_WEP_hook'() {bLoopAnimation=false, sequence=&GrappleKnife_GroundAttack_02_male_WEP_hook})
  ]
 }
 CastShort {
  path=AnimGraph.AnimGraphNode_StateMachine_0.Grapple Anim.AnimStateNode_4.CastShort
  nodes [Output Animation Pose(Result=Sequence Player 'GrappleKnife_Cast_Short_Fullbody_WEP_Hook'() {bLoopAnimation=false, sequence=/GrappleWeapon/Items/Animations/WEP/GrappleKnife_Cast_Short_Fullbody_WEP_Hook})]
 }
 UpperBodyAttack_1 {
  path=AnimGraph.AnimGraphNode_StateMachine_0.Grapple Anim.AnimStateNode_6.UpperBodyAttack_1
  nodes [Output Animation Pose(Result=Sequence Player 'GrappleKnife_GroundAttack_01_UpperbodyOnly_WEP_hook'() {bLoopAnimation=false, sequence=/GrappleWeapon/Items/Animations/WEP/GrappleKnife_GroundAttack_01_UpperbodyOnly_WEP_hook})]
 }
 UpperBodyAttack_2 {
  path=AnimGraph.AnimGraphNode_StateMachine_0.Grapple Anim.AnimStateNode_7.UpperBodyAttack_2
  nodes [Output Animation Pose(Result=Sequence Player 'GrappleKnife_GroundAttack_02_UpperbodyOnly_WEP_hook'() {bLoopAnimation=false, sequence=/GrappleWeapon/Items/Animations/WEP/GrappleKnife_GroundAttack_02_UpperbodyOnly_WEP_hook})]
 }
 CastLong {
  path=AnimGraph.AnimGraphNode_StateMachine_0.Grapple Anim.AnimStateNode_8.CastLong
  nodes [Output Animation Pose(Result=Sequence Player 'GrappleKnife_Cast_Long_Fullbody_WEP_Hook'() {bLoopAnimation=false, sequence=/GrappleWeapon/Items/Animations/WEP/GrappleKnife_Cast_Long_Fullbody_WEP_Hook})]
 }
 CastLMid {
  path=AnimGraph.AnimGraphNode_StateMachine_0.Grapple Anim.AnimStateNode_9.CastLMid
  nodes [Output Animation Pose(Result=Sequence Player 'GrappleKnife_Cast_Long_Fullbody_WEP_Hook'() {bLoopAnimation=false, sequence=&GrappleKnife_Cast_Long_Fullbody_WEP_Hook})]
 }
 Transition {path=AnimGraph.AnimGraphNode_StateMachine_0.Grapple Anim.AnimStateTransitionNode_0.Transition, nodes [Result()]}
 Transition {
  path=AnimGraph.AnimGraphNode_StateMachine_0.Grapple Anim.AnimStateTransitionNode_1.Transition
  nodes [Result(bCanEnterTransition=Get MantisAttack_3() -> (MantisAttack_3:bool))]
 }
 Transition {path=AnimGraph.AnimGraphNode_StateMachine_0.Grapple Anim.AnimStateTransitionNode_10.Transition, nodes [Result(bCanEnterTransition=true)]}
 Transition {path=AnimGraph.AnimGraphNode_StateMachine_0.Grapple Anim.AnimStateTransitionNode_11.Transition, nodes [Result()]}
 Transition {path=AnimGraph.AnimGraphNode_StateMachine_0.Grapple Anim.AnimStateTransitionNode_12.Transition, nodes [Result()]}
 Transition {
  path=AnimGraph.AnimGraphNode_StateMachine_0.Grapple Anim.AnimStateTransitionNode_13.Transition
  nodes [Result(bCanEnterTransition=Get MantisAttack_4() -> (MantisAttack_4:bool))]
 }
 Transition {
  path=AnimGraph.AnimGraphNode_StateMachine_0.Grapple Anim.AnimStateTransitionNode_14.Transition
  nodes [Result(bCanEnterTransition=Get MantisAttack_3())]
 }
 Transition {path=AnimGraph.AnimGraphNode_StateMachine_0.Grapple Anim.AnimStateTransitionNode_15.Transition, nodes [Result(bCanEnterTransition=true)]}
 Transition {path=AnimGraph.AnimGraphNode_StateMachine_0.Grapple Anim.AnimStateTransitionNode_16.Transition, nodes [Result(bCanEnterTransition=true)]}
 Transition {path=AnimGraph.AnimGraphNode_StateMachine_0.Grapple Anim.AnimStateTransitionNode_17.Transition, nodes [Result()]}
 Transition {path=AnimGraph.AnimGraphNode_StateMachine_0.Grapple Anim.AnimStateTransitionNode_18.Transition, nodes [Result()]}
 Transition {
  path=AnimGraph.AnimGraphNode_StateMachine_0.Grapple Anim.AnimStateTransitionNode_2.Transition
  nodes [Result(bCanEnterTransition=Get MantisAttack_2() -> (MantisAttack_2:bool))]
 }
 Transition {path=AnimGraph.AnimGraphNode_StateMachine_0.Grapple Anim.AnimStateTransitionNode_27.Transition, nodes [Result()]}
 Transition {
  path=AnimGraph.AnimGraphNode_StateMachine_0.Grapple Anim.AnimStateTransitionNode_28.Transition
  nodes [Result(bCanEnterTransition=Get!() -> (IsAttackFromCast:bool, IsAttackFromCast_b:bool).IsAttackFromCast_b)]
 }
 Transition {
  path=AnimGraph.AnimGraphNode_StateMachine_0.Grapple Anim.AnimStateTransitionNode_3.Transition
  nodes [Result(bCanEnterTransition=Get MantisAttack_1() -> (MantisAttack_1:bool))]
 }
 Transition {
  path=AnimGraph.AnimGraphNode_StateMachine_0.Grapple Anim.AnimStateTransitionNode_32.Transition
  nodes [Result(bCanEnterTransition=Get IsHookFireShort() -> (IsHookFireShort:bool))]
 }
 Transition {
  path=AnimGraph.AnimGraphNode_StateMachine_0.Grapple Anim.AnimStateTransitionNode_33.Transition
  nodes [Result(bCanEnterTransition=Get IsHookFireMid() -> (IsHookFireMid:bool))]
 }
 Transition {
  path=AnimGraph.AnimGraphNode_StateMachine_0.Grapple Anim.AnimStateTransitionNode_34.Transition
  nodes [Result(bCanEnterTransition=Get IsHookFireLong() -> (IsHookFireLong:bool))]
 }
 Transition {
  path=AnimGraph.AnimGraphNode_StateMachine_0.Grapple Anim.AnimStateTransitionNode_35.Transition
  nodes [Result(bCanEnterTransition=Get IsAttackFromCast() -> (IsAttackFromCast:bool))]
 }
 Transition {
  path=AnimGraph.AnimGraphNode_StateMachine_0.Grapple Anim.AnimStateTransitionNode_36.Transition
  nodes [Result(bCanEnterTransition=Get IsAttackFromCast())]
 }
 Transition {
  path=AnimGraph.AnimGraphNode_StateMachine_0.Grapple Anim.AnimStateTransitionNode_37.Transition
  nodes [Result(bCanEnterTransition=Get IsAttackFromCast())]
 }
 Transition {
  path=AnimGraph.AnimGraphNode_StateMachine_0.Grapple Anim.AnimStateTransitionNode_4.Transition
  nodes [Result(bCanEnterTransition=Get MantisAttackAir() -> (MantisAttackAir:bool))]
 }
 Transition {path=AnimGraph.AnimGraphNode_StateMachine_0.Grapple Anim.AnimStateTransitionNode_5.Transition, nodes [Result()]}
 Transition {
  path=AnimGraph.AnimGraphNode_StateMachine_0.Grapple Anim.AnimStateTransitionNode_6.Transition
  nodes [Result(bCanEnterTransition=Get MantisAttack_1())]
 }
 Transition {
  path=AnimGraph.AnimGraphNode_StateMachine_0.Grapple Anim.AnimStateTransitionNode_7.Transition
  nodes [Result(bCanEnterTransition=Get MantisAttack_4())]
 }
 Transition {
  path=AnimGraph.AnimGraphNode_StateMachine_0.Grapple Anim.AnimStateTransitionNode_8.Transition
  nodes [Result(bCanEnterTransition=Get MantisAttack_2())]
 }
 Transition {path=AnimGraph.AnimGraphNode_StateMachine_0.Grapple Anim.AnimStateTransitionNode_9.Transition, nodes [Result(bCanEnterTransition=true)]}
 EventGraph {nodes [
   Event BlueprintUpdateAnimation()
   n1 = TryGetPawnOwner()
   > Is Valid(InputObject=@n1)
   > n2 = Cast To PlayerPawn_Athena(Object=@n1)
   > n3 = Cast To B_GrappleKnife!(execute, Object=Get CurrentWeapon!(self:Fort Pawn Object Reference=@n2) -> (CurrentWeapon:Fort Weapon Object Reference)) -> (then, CastFailed, AsB Grapple Knife:B Grapple Knife Object Reference) {is=title K2Node_DynamicCast}
   > Set IsFired(IsFired:bool=Get IsFired!(self:B Grapple Knife Object Reference=@n3) -> (IsFired:bool)) -> (Output_Get:bool)
   > Set DistanceBucket(DistanceBucket:int=Get DistanceBucket!(self:B Grapple Knife Object Reference=@n3) -> (DistanceBucket:int=0)) -> (Output_Get:int=0)
   > Set Is Grapple Armed(Is Grapple Armed:bool=Get IsGrappleArmed!(self:B Grapple Knife Object Reference=@n3) -> (IsGrappleArmed:bool)) -> (Output_Get:bool)
  ]}
]
variables [Is Grapple Armed {type=boolean}]
